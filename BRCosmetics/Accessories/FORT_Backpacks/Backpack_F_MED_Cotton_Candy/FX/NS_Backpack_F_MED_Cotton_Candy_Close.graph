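# Niagara one-shot burst effect for the Cotton Candy backpack, apparently played when it closes.
# A brief blue Glow sprite drops downward, then delayed bursts of Smoke puffs and GPU Sparkles follow.
# Smoke and Sparkles use two spawn groups to split into opposite sideways, upward sprays that fall under drag and gravity.
# Smoke animates SubUV frames while shifting to violet and fading; magenta-violet sparkles swirl with curl noise and flicker through a jagged size curve.
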NiagaraSystem /BRCosmetics/Accessories/FORT_Backpacks/Backpack_F_MED_Cotton_Candy/FX/NS_Backpack_F_MED_Cotton_Candy_Close
dependencies {
 usedDataInterfaces [&NiagaraDataInterfaceCurve, &NiagaraDataInterfaceColorCurve]
 usedDynamicInputs []
 usedModules [&SystemState]
}
emitters {
 Glow {
  emitterUpdateScript [
   &EmitterState()
   &SpawnBurst_Instantaneous(Spawn Count=1, Use Spawn Probability=false)
  ]
  particleSpawnScript [
   &InitializeParticle(
    Color=<rgba 0 0.4167 1 1>
    Lifetime=0.15
    Sprite Size=<100 150>
    Sprite Size Mode=::NewEnumerator4
    UsePositionOffset=false
    Write Material Random=true
   )
   &SystemLocation(MaskSpawnGroup=false, Offset=<0 -5 30>)
   &AddVelocity(Velocity=<0 0 -130>)
  ]
  particleUpdateScript [
   &ParticleState()
   &ScaleSpriteSize(Scale Factor=&Vector2DFromFloat(Value=&FloatFromCurve(FloatCurve=DataInterface() {propertyValues {Curve {keys [
          {interpMode=RCIM_Cubic, leaveTangent=0.2376, tangentMode=RCTM_Break, time=0, value=0}
          {arriveTangent=1.3815, interpMode=RCIM_Cubic, leaveTangent=1.3815, tangentMode=RCTM_User, time=0.7, value=1}
          {arriveTangent=-1.4846, interpMode=RCIM_Cubic, tangentMode=RCTM_Break, time=1, value=0}
         ]}}})))
   &CameraOffset(Camera Offset Amount=15)
   &Drag(Drag=2, Use Linear Drag=true, Use Rotational Drag=false)
   &SolveForcesAndVelocity(Clamp Velocity=false, Limit Acceleration=false)
   &ScaleColor(
    Scale RGB=&VectorFromFloat(Value=&FloatFromCurve(FloatCurve=DataInterface() {propertyValues {Curve {keys [<at 0 0>,<at 1 1>]}}}))
    ScaleA=true
    ScaleRGB=true
   )
  ]
  rendererClasses [&NiagaraSpriteRendererProperties]
  renderers [{rendererClass=&NiagaraSpriteRendererProperties}]
 }
 Smoke {
  emitterUpdateScript [
   &EmitterState()
   &SpawnBurst_Instantaneous(Spawn Count=6, Spawn Time=0.1, Use Spawn Probability=false)
  ]
  particleSpawnScript [
   /Niagara/Modules/Spawn/Initialization/InitializeParticle(
    Color=&RandomRangeLinearColor(Minimum=<rgba 0.4688 0.4688 0.4688 0.8>, Maximum=<rgba 0.4688 0.4688 0.4688 0.9>)
    Lifetime=&RandomRangeFloat(Minimum=1, Maximum=1.5)
    Module.Write Color=true
    Module.Write Lifetime=true
    Module.Write Mass=true
    Module.Write Material Random=true
    Module.Write Position=true
    Module.Write Scale=false
    Module.Write SpriteRotation=true
    Module.Write SpriteSize=true
    Sprite Rotation=&RandomRangeFloat(Minimum=0, Maximum=360)
    Sprite Size=&Vector2DFromFloat(Value=&RandomRangeFloat(Minimum=30, Maximum=40))
   )
   &SphereLocation(
    MaskSpawnGroup=false
    Offset=<0 -5 6>
    Override Seed=false
    Sphere Radius=2
    Surface Only=false
    UseNonUniformScale=false
    UseOffset=true
    UseSphereOrientation=false
   )
   &AddVelocityFromPoint(Module.Use Falloff Distance=false, Velocity Strength=&RandomRangeFloat(Minimum=10, Maximum=60))
   &FN_SetSpawnGroup(Number of Spawn Groups=2)
   &self:Add_Velocity_From_Spawn_Group(
    Velocity:Vector3f=&RandomRangeVector(Minimum=<100 0 80>, Maximum=<30 0 60>)
    Velocity Second:Vector3f=&RandomRangeVector(Minimum=<-100 0 80>, Maximum=<-30 0 60>)
   ) {module=Add_Velocity_From_Spawn_Group}
  ]
  particleUpdateScript [
   &ParticleState()
   &Drag(Drag=&RandomRangeFloat(Minimum=3.5, Maximum=3), Use Linear Drag=true, Use Rotational Drag=false)
   &GravityForce(Gravity=&RandomRangeVector(Minimum=<0 0 -100>, Maximum=<0 0 -200>))
   &SolveForcesAndVelocity(Clamp Velocity=false, Limit Acceleration=false)
   &SubUVAnimation(
    End Frame=50
    Start Frame=10
    SubUV Animation Mode:ENiagara_SubUVLookupMode=::NewEnumerator1
    SubUV Curve=DataInterface() {propertyValues {Curve {keys [
        {arriveTangent=1.5272, interpMode=RCIM_Cubic, leaveTangent=1.5272, tangentMode=RCTM_User, time=0, value=0}
        {arriveTangent=0.0067, interpMode=RCIM_Cubic, leaveTangent=0.0067, tangentMode=RCTM_User, time=1, value=1}
       ]}}}
   )
   &ScaleSpriteSize(Scale Factor=&Vector2DFromFloat(Value=&FloatFromCurve(FloatCurve=DataInterface() {propertyValues {Curve {keys [
          {arriveTangent=0.0059, interpMode=RCIM_Cubic, leaveTangent=0.0059, tangentMode=RCTM_User, time=0, value=1}
          {arriveTangent=0.1826, interpMode=RCIM_Cubic, leaveTangent=0.1826, tangentMode=RCTM_User, time=1, value=1.8}
         ]}}})))
   &ScaleColor(
    Linear Color Curve=DataInterface() {propertyValues {
      AlphaCurve {keys [<at 0 1>,<at 1 0>]}
      BlueCurve {keys [<at 0.0013 0.5>,<at 0.435 1>,<at 1 1>]}
      GreenCurve {keys [<at 0.0013 0.2708>,<at 0.435 0.1>,<at 1 0>]}
      RedCurve {keys [<at 0.0013 0.25>,<at 0.435 0.25>,<at 1 0.4167>]}
     }}
    Scale Mode=::NewEnumerator2
   ) {module=ScaleColor001}
   &ScaleColor(
    Scale Alpha=&FloatFromCurve(FloatCurve=DataInterface() {propertyValues {Curve {keys [
         {interpMode=RCIM_Cubic, tangentMode=RCTM_User, time=0, value=1.2}
         {arriveTangent=-1.7719, interpMode=RCIM_Cubic, leaveTangent=-1.7719, tangentMode=RCTM_User, time=0.8, value=0.75}
         {arriveTangent=-4.4359, interpMode=RCIM_Cubic, tangentMode=RCTM_Break, time=1, value=0}
        ]}}})
    Scale RGB=&VectorFromFloat(Value=&FloatFromCurve(FloatCurve=DataInterface() {propertyValues {Curve {keys [
          {arriveTangent=-1.4923, interpMode=RCIM_Cubic, leaveTangent=-1.4923, tangentMode=RCTM_User, time=0, value=1.5}
          {arriveTangent=-1.2406, interpMode=RCIM_Cubic, leaveTangent=-1.2406, tangentMode=RCTM_User, time=0.9, value=0.5}
          {arriveTangent=-1.9627, interpMode=RCIM_Cubic, leaveTangent=-1.9627, tangentMode=RCTM_User, time=1, value=0}
         ]}}}))
    ScaleA=true
    ScaleRGB=true
   )
   &CameraOffset(Camera Offset Amount=&RandomRangeFloat(Minimum=0, Maximum=5))
  ]
  rendererClasses [&NiagaraSpriteRendererProperties]
  renderers [{rendererClass=&NiagaraSpriteRendererProperties}]
 }
 Sparkles {
  simTarget=GPUComputeSim
  emitterUpdateScript [
   &EmitterState()
   &SpawnBurst_Instantaneous(Spawn Count=100, Spawn Time=0.1, Use Spawn Probability=false)
  ]
  particleSpawnScript [
   &InitializeParticle(
    Color=&HueShiftLinearColor(
     Hue=0
     Saturation=&RandomRangeFloat(Minimum=5, Maximum=10)
     Value=&RandomRangeFloat(Minimum=1, Maximum=5)
     Linear Color=&RandomRangeLinearColor(Minimum=<rgba 0.8 0 0.8 1>, Maximum=<rgba 0.1333 0 0.8 1>, Color Channel Mode=::NewEnumerator2)
    )
    Lifetime=&Power(a=&RandomRangeFloat(Minimum=0.46, Maximum=1), b=2)
    Module.Write Color=true
    Module.Write Lifetime=true
    Module.Write Mass=true
    Module.Write Material Random=true
    Module.Write Position=false
    Module.Write Scale=false
    Module.Write SpriteRotation=true
    Module.Write SpriteSize=true
    Position=&SimulationPosition()
    Sprite Rotation=&RandomRangeFloat(Minimum=0, Maximum=360)
    Sprite Size=&Vector2DFromFloat(Value=&RandomRangeFloat(Minimum=1, Maximum=5))
   )
   &SphereLocation(
    MaskSpawnGroup=false
    Override Seed=false
    Sphere Radius=&RandomRangeFloat(Minimum=2, Maximum=10)
    Surface Only=false
    UseNonUniformScale=false
    UseOffset=false
    UseSphereOrientation=false
   )
   &AddVelocityFromPoint(Module.Use Falloff Distance=false, Velocity Strength=&RandomRangeFloat(Minimum=50, Maximum=150))
   &FN_SetSpawnGroup(Number of Spawn Groups=2)
   &self:Add_Velocity_From_Spawn_Group(Velocity:Vector3f=<100 0 200>, Velocity Second:Vector3f=<-100 0 200>) {module=Add_Velocity_From_Spawn_Group}
  ]
  particleUpdateScript [
   &ParticleState()
   &Drag(Drag=5, Use Linear Drag=true, Use Rotational Drag=false)
   &GravityForce(Gravity=&RandomRangeVector(Minimum=<0 0 -200>, Maximum=<1 1 -500>))
   /Niagara/Modules/Update/Forces/CurlNoiseForce(Noise Strength=200, Pan Noise=false)
   &SolveForcesAndVelocity(Clamp Velocity=false, Limit Acceleration=false)
   &ScaleColor(
    Scale Alpha=5
    Scale RGB=&VectorFromFloat(Value=&RandomRangeFloat(Minimum=0.04, Maximum=0.01))
    ScaleA=true
    ScaleRGB=true
   )
   &ScaleSpriteSize(Scale Factor=&Vector2DFromFloat(Value=&FloatFromCurve(FloatCurve=DataInterface() {propertyValues {Curve {keys [<at 0.0018 0.1957>,<at 0.0527 0.819>,<at 0.1169 0.6555>,<at 0.1748 0.9015>,<at 0.2394 0.582>,<at 0.332 1.0023>,<at 0.4947 0.3707>,<at 0.6727 1>,<at 1 0>]}}})))
  ]
  rendererClasses [&NiagaraSpriteRendererProperties]
  renderers [{rendererClass=&NiagaraSpriteRendererProperties}]
 }
}
summary {
 systemName=NS_Backpack_F_MED_Cotton_Candy_Close
 emitters [
  {emitterName=Glow, rendererClasses [&NiagaraSpriteRendererProperties]}
  {emitterName=Sparkles, simTarget=GPUComputeSim, rendererClasses [&NiagaraSpriteRendererProperties]}
  {emitterName=Smoke, rendererClasses [&NiagaraSpriteRendererProperties]}
 ]
 userVariables [User.RibbonIDOffset {type=NiagaraInt32}]
}
