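# Niagara one-shot "surprise" burst for the Cotton Candy backpack cosmetic, popping confetti, streamers, smoke, sparkles and fireworks.
# Glow, light, flash stars, smoke, fragments, strips and GPU sparkles burst near the origin, most split by spawn groups into opposite sideways sprays.
# FireworkParent sends location events, apparently feeding FireworkChild trails.
# Ribbon streamers share per-ribbon seeds for direction and hue, and flash stars scale with time of day.
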NiagaraSystem /BRCosmetics/Accessories/FORT_Backpacks/Backpack_F_MED_Cotton_Candy/FX/NS_Backpack_F_MED_Cotton_Candy_Surprise
dependencies {
 usedDataInterfaces [&NiagaraDataInterfaceCurve, &NiagaraDataInterfaceColorCurve]
 usedDynamicInputs []
 usedModules [&SystemState]
}
emitters {
 FireworkChild {
  emitterUpdateScript [&EmitterState(Life Cycle Mode=::NewEnumerator1, Loop Behavior=::NewEnumerator1, Loop Duration=1.2, UseLoopDelay=false)]
  particleSpawnScript [/Niagara/Modules/Spawn/Initialization/InitializeParticle(
    Color=&MakeLinearColorFromVectorAndFloat("Vector (RGB)"=&MakeVectorFromLinearColorRGB(Color=<rgba 2 18.4999 200 1>), "Float (Alpha)"=1)
    Lifetime=&Lerp_Float(A=0.5, B=1.4, Alpha=&Power(a=&RandomRangeFloat(Minimum=0, Maximum=1), b=2))
    Module.Write Color=true
    Module.Write Lifetime=true
    Module.Write Mass=true
    Module.Write Material Random=true
    Module.Write Position=false
    Module.Write Scale=false
    Module.Write SpriteRotation=false
    Module.Write SpriteSize=true
    Position=&SimulationPosition()
    Sprite Size=&Vector2DFromFloat(Value=&RandomRangeFloat(Minimum=0.4, Maximum=0.3))
   )]
  particleUpdateScript [
   &ParticleState()
   &Drag(Use Linear Drag=true, Use Rotational Drag=false)
   &GravityForce(Gravity=&RandomRangeVector(Minimum=<0 0 -100>, Maximum=<1 1 -200>))
   &SolveForcesAndVelocity(Clamp Velocity=false, Limit Acceleration=false)
   &ScaleColor(
    Scale Alpha=&UniformAOrBFloat(A=1, B=&RandomRangeFloat(Minimum=1, Maximum=300, Evaluation Type=::NewEnumerator1))
    Scale RGB=&MakeVectorFromLinearColorRGB(Color=&ColorFromCurve(ColorCurve=DataInterface() {propertyValues {
        AlphaCurve {keys [<at 0 1 RCIM_Cubic>,<at 1 0 RCIM_Cubic>]}
        BlueCurve {keys [<at 0.0228 0.2 RCIM_Cubic>,<at 0.4725 0.0499>]}
        GreenCurve {keys [<at 0.0228 1 RCIM_Cubic>,<at 0.4725 0.2914>]}
        RedCurve {keys [<at 0.0228 1 RCIM_Cubic>,<at 0.4725 1>]}
       }}))
    ScaleA=true
    ScaleRGB=true
   )
   &ScaleSpriteSize(Scale Factor=&Vector2DFromFloat(Value=&FloatFromCurve(
      FloatCurve=DataInterface() {propertyValues {Curve {keys [
          {arriveTangent=0.0859, interpMode=RCIM_Cubic, leaveTangent=0.0859, tangentMode=RCTM_User, time=0, value=1}
          {arriveTangent=-3.6686, interpMode=RCIM_Cubic, leaveTangent=-3.6686, tangentMode=RCTM_User, time=1, value=0}
         ]}}}
      Scale Curve=&RandomRangeFloat(Minimum=0.5, Maximum=2, Evaluation Type=::NewEnumerator1)
     )))
  ]
  rendererClasses [&NiagaraSpriteRendererProperties]
  renderers [{rendererClass=&NiagaraSpriteRendererProperties}]
 }
 FireworkParent {
  emitterUpdateScript [
   &EmitterState(Life Cycle Mode=::NewEnumerator1, Loop Behavior=::NewEnumerator1, Loop Duration=1.5, UseLoopDelay=false)
   &SpawnBurst_Instantaneous(Spawn Count=8, Use Spawn Probability=false)
  ]
  particleSpawnScript [
   &InitializeParticle(
    Lifetime=&RandomRangeFloat(Minimum=0.1, Maximum=0.3)
    Module.Write Color=true
    Module.Write Lifetime=true
    Module.Write Mass=true
    Module.Write Material Random=true
    Module.Write Position=false
    Module.Write Scale=false
    Module.Write SpriteRotation=false
    Module.Write SpriteSize=true
    Position=&SimulationPosition()
    Sprite Size=&Vector2DFromFloat(Value=0)
   )
   &SphereLocation(
    MaskSpawnGroup=false
    Offset=<0 -5 10>
    Override Seed=false
    Sphere Radius=1
    Surface Only=false
    UseNonUniformScale=false
    UseOffset=true
    UseSphereOrientation=false
   )
   &AddVelocityFromPoint(Module.Use Falloff Distance=false, Velocity Strength=&RandomRangeFloat(Minimum=0, Maximum=150))
   &FN_SetSpawnGroup(Number of Spawn Groups=2)
   &self:Add_Velocity_From_Spawn_Group(
    Velocity:Vector3f=&RandomRangeVector(Minimum=<250 50 250>, Maximum=<200 -50 150>)
    Velocity Second:Vector3f=&RandomRangeVector(Minimum=<-250 50 250>, Maximum=<-200 -50 150>)
   ) {module=Add_Velocity_From_Spawn_Group}
  ]
  particleUpdateScript [
   &ParticleState()
   &Drag(Drag=3.5, Use Linear Drag=true, Use Rotational Drag=false)
   &GravityForce()
   &SolveForcesAndVelocity(Clamp Velocity=false, Limit Acceleration=false)
   &GenerateLocationEvent(Event Send Rate=50, Module.Use Event Send Rate=true)
  ]
  rendererClasses [&NiagaraSpriteRendererProperties]
  renderers [{rendererClass=&NiagaraSpriteRendererProperties}]
 }
 FlashStars {
  emitterUpdateScript [
   &EmitterState(Life Cycle Mode=::NewEnumerator1, Loop Behavior=::NewEnumerator1, Loop Delay=0.05, Loop Duration=1, UseLoopDelay=true)
   &SpawnBurst_Instantaneous(
    Spawn Count=&RandomRangeInt(Minimum=4, Maximum=6, Evaluation Type=::NewEnumerator1)
    Use Spawn Probability=false
   )
  ]
  particleSpawnScript [
   &FN_SetSpawnGroup(Number of Spawn Groups=2)
   /Niagara/Modules/Spawn/Initialization/V2/InitializeParticle(
    Color=<rgba 10 10 2 1>
    Lifetime=&RandomRangeFloat(Minimum=0.3, Maximum=0.7)
    Sprite Size=&Multiply_Vector2(
     A=&RandomRangeVector2D(Minimum=<25 15>, Maximum=<45 35>)
     B=&Vector2DFromFloat(Value=&RandomRangeFloat(Minimum=0.5, Maximum=0.2))
    )
    Sprite Size Mode=::NewEnumerator4
    UsePositionOffset=false
    Write Material Random=true
   )
   /SetVariables(Particles.InitialSize:Vector2f=&Vector2DFromFloat(Value=&RandomRangeFloat(Minimum=10, Maximum=30))) {isSetParameters=true}
   /SetVariables(Particles.SpriteRotation2:NiagaraFloat=&RandomRangeFloat(Minimum=0, Maximum=360)) {isSetParameters=true}
   &SphereLocation(
    MaskSpawnGroup=false
    Offset=<0 -5 5>
    Override Seed=false
    Sphere Radius=4
    Surface Only=false
    UseNonUniformScale=false
    UseOffset=true
    UseSphereOrientation=false
   )
   &AddVelocityFromPoint(Module.Use Falloff Distance=false, Velocity Strength=&RandomRangeFloat(Minimum=50, Maximum=250))
   &self:Add_Velocity_From_Spawn_Group(
    Velocity:Vector3f=&RandomRangeVector(Minimum=<150 0 300>, Maximum=<60 0 150>)
    Velocity Second:Vector3f=&RandomRangeVector(Minimum=<-150 0 300>, Maximum=<-60 0 150>)
   ) {module=Add_Velocity_From_Spawn_Group}
   &CameraOffset(Camera Offset Amount=&RandomRangeFloat(Minimum=20, Maximum=40))
  ]
  particleUpdateScript [
   &Drag(Drag=&RandomRangeFloat(Minimum=10, Maximum=15), Use Linear Drag=true, Use Rotational Drag=false)
   &GravityForce(Gravity=&RandomRangeVector(Minimum=<0 0 -200>, Maximum=<1 1 -300>))
   &SolveForcesAndVelocity(Clamp Velocity=false, Limit Acceleration=false)
   &ParticleState()
   /SetVariables(Particles.TimeOfDay:NiagaraFloat=&FNTimeOfDayScalar(Night Scale=0.5, Morning Scale=0.3, Evening Scale=0.3)) {isSetParameters=true}
   /SetVariables(Particles.Color2=&MultiplyLinearColor(
     Linear Color=&ColorFromCurve(ColorCurve=DataInterface() {propertyValues {
        AlphaCurve {keys [<at 0 0>,<at 0.15 0.2 RCIM_Cubic>,<at 1 0 RCIM_Cubic>]}
        BlueCurve {keys [<at 0.1483 0.15 RCIM_Cubic>,<at 1 0.0156 RCIM_Cubic>]}
        GreenCurve {keys [<at 0.1483 0.7167 RCIM_Cubic>,<at 1 0.4258 RCIM_Cubic>]}
        RedCurve {keys [<at 0.1483 1 RCIM_Cubic>,<at 1 1 RCIM_Cubic>]}
       }})
     Scale Factor=&MakeFloat4FromVectorAndFloat("Vector (RGB)"=<1 1 1>, "Float (Alpha)"=$Particles.TimeOfDay)
    )) {isSetParameters=true}
   /SetVariables(Particles.SpriteSize2=&Multiply_Vector2(
     A=$Particles.InitialSize
     B=&Vector2DFromFloat(Value=&FloatFromCurve(FloatCurve=DataInterface() {propertyValues {Curve {keys [
           <at 0 0 RCIM_Cubic>
           {arriveTangent=-0.0175, interpMode=RCIM_Cubic, leaveTangent=-0.0175, tangentMode=RCTM_User, time=0.1563, value=1.5}
           {arriveTangent=-0.0042, interpMode=RCIM_Cubic, leaveTangent=-0.0042, tangentMode=RCTM_User, time=0.2646, value=0.0384}
           {arriveTangent=0.0081, interpMode=RCIM_Cubic, leaveTangent=0.0081, tangentMode=RCTM_User, time=0.4692, value=0.7123}
           {arriveTangent=0.034, interpMode=RCIM_Cubic, leaveTangent=0.034, tangentMode=RCTM_User, time=0.6406, value=0.0348}
           {arriveTangent=0.019, interpMode=RCIM_Cubic, leaveTangent=0.019, tangentMode=RCTM_User, time=0.7993, value=0.5053}
           <at 1 0 RCIM_Cubic>
          ]}}}))
    )) {isSetParameters=true}
   &ScaleSpriteSize(Scale Factor=&Vector2DFromFloat(Value=&FloatFromCurve(FloatCurve=DataInterface() {propertyValues {Curve {keys [
          <at 0 0 RCIM_Cubic>
          {arriveTangent=-0.0175, interpMode=RCIM_Cubic, leaveTangent=-0.0175, tangentMode=RCTM_User, time=0.1563, value=1.5}
          {arriveTangent=-0.0042, interpMode=RCIM_Cubic, leaveTangent=-0.0042, tangentMode=RCTM_User, time=0.2646, value=0.0384}
          {arriveTangent=0.0081, interpMode=RCIM_Cubic, leaveTangent=0.0081, tangentMode=RCTM_User, time=0.4692, value=0.7123}
          {arriveTangent=0.034, interpMode=RCIM_Cubic, leaveTangent=0.034, tangentMode=RCTM_User, time=0.6406, value=0.0348}
          {arriveTangent=0.019, interpMode=RCIM_Cubic, leaveTangent=0.019, tangentMode=RCTM_User, time=0.7993, value=0.5053}
          <at 1 0 RCIM_Cubic>
         ]}}}))) {module=ScaleSpriteSize001}
   &ScaleColor(
    Scale Alpha=$Particles.TimeOfDay
    Scale RGB=&Multiply_Vector(
     A=<1 1 1>
     B=&VectorFromFloat(Value=&FloatFromCurve(FloatCurve=DataInterface() {propertyValues {Curve {keys [<at 0 0>,<at 0.1 1>,<at 1 0>]}}}, Scale Curve=5))
    )
    ScaleA=true
    ScaleRGB=true
   )
  ]
  rendererClasses [&NiagaraSpriteRendererProperties]
  renderers [
   {rendererClass=&NiagaraSpriteRendererProperties}
   {rendererClass=&NiagaraSpriteRendererProperties, rendererIndex=1}
  ]
 }
 Fragments {
  emitterUpdateScript [
   &EmitterState()
   &SpawnBurst_Instantaneous(Spawn Count=40, Use Spawn Probability=false)
  ]
  particleSpawnScript [
   /Niagara/Modules/Spawn/Initialization/InitializeParticle(
    Color=&RandomRangeLinearColor(Minimum=<rgba 0.62 0 0 1>, Maximum=<rgba 1 0 0 1>)
    Lifetime=&RandomRangeFloat(Minimum=0.2, Maximum=1.2)
    Module.Write Color=true
    Module.Write Lifetime=true
    Module.Write Mass=true
    Module.Write Material Random=true
    Module.Write Position=false
    Module.Write Scale=false
    Module.Write SpriteRotation=true
    Module.Write SpriteSize=true
    Position=&SimulationPosition()
    Sprite Rotation=&RandomRangeFloat(Minimum=0, Maximum=360)
    Sprite Size=&Vector2DFromFloat(Value=&RandomRangeFloat(Minimum=2, Maximum=4.5))
   )
   &SphereLocation(
    MaskSpawnGroup=false
    Offset=<0 -5 8>
    Override Seed=false
    Sphere Radius=&RandomRangeFloat(Minimum=0, Maximum=10)
    Surface Only=false
    UseNonUniformScale=false
    UseOffset=true
    UseSphereOrientation=false
   )
   &AddVelocityFromPoint(Module.Use Falloff Distance=false, Velocity Strength=&RandomRangeFloat(Minimum=50, Maximum=200))
   &FN_SetSpawnGroup(Number of Spawn Groups=2)
   &self:Add_Velocity_From_Spawn_Group(Velocity:Vector3f=<100 0 50>, Velocity Second:Vector3f=<-100 0 50>) {module=Add_Velocity_From_Spawn_Group}
  ]
  particleUpdateScript [
   &ParticleState()
   &Drag(Drag=5, Use Linear Drag=true, Use Rotational Drag=false)
   &GravityForce(Gravity=&RandomRangeVector(Minimum=<0 0 -250>, Maximum=<0 0 -150>))
   /Niagara/Modules/Update/Forces/CurlNoiseForce(Noise Strength=150, Pan Noise=false)
   &SolveForcesAndVelocity(Clamp Velocity=false, Limit Acceleration=false)
   &SpriteRotationRate(Rotation Rate=&RandomRangeFloat(Minimum=-500, Maximum=500))
   &ScaleSpriteSize(Scale Factor=&Vector2DFromFloat(Value=&FloatFromCurve(FloatCurve=DataInterface() {propertyValues {Curve {keys [
          {arriveTangent=-0.2862, interpMode=RCIM_Cubic, leaveTangent=-0.2862, tangentMode=RCTM_User, time=0.55, value=1}
          {arriveTangent=-7.3879, interpMode=RCIM_Cubic, leaveTangent=-7.3879, tangentMode=RCTM_User, time=1, value=0}
         ]}}})))
   &ScaleColor(Scale Alpha=2, Scale RGB=&VectorFromFloat(Value=1), ScaleA=true, ScaleRGB=true)
  ]
  rendererClasses [&NiagaraSpriteRendererProperties]
  renderers [{rendererClass=&NiagaraSpriteRendererProperties}]
 }
 Glow {
  emitterUpdateScript [
   &EmitterState()
   &SpawnBurst_Instantaneous(Spawn Count=1, Use Spawn Probability=false)
  ]
  particleSpawnScript [
   /Niagara/Modules/Spawn/Initialization/V2/InitializeParticle(
    Color=<rgba 0 0.3333 1 1>
    Lifetime=0.18
    Sprite Size=<200 200>
    Sprite Size Mode=::NewEnumerator4
    UsePositionOffset=false
    Write Material Random=true
   )
   &SystemLocation(MaskSpawnGroup=false, Offset=<0 -5 5>)
   &AddVelocity(Velocity=<0 0 200>)
  ]
  particleUpdateScript [
   &ParticleState()
   &ScaleSpriteSize(Scale Factor=&Vector2DFromFloat(Value=&FloatFromCurve(FloatCurve=DataInterface() {propertyValues {Curve {keys [
          {interpMode=RCIM_Cubic, leaveTangent=0.2376, tangentMode=RCTM_Break, time=0, value=0.5}
          {arriveTangent=0.0486, interpMode=RCIM_Cubic, leaveTangent=0.0486, tangentMode=RCTM_User, time=0.15, value=1}
          {arriveTangent=-2.7369, interpMode=RCIM_Cubic, tangentMode=RCTM_Break, time=1, value=0}
         ]}}})))
   &CameraOffset(Camera Offset Amount=15)
   &Drag(Drag=4, Use Linear Drag=true, Use Rotational Drag=false)
   &SolveForcesAndVelocity(Clamp Velocity=false, Limit Acceleration=false)
  ]
  rendererClasses [&NiagaraSpriteRendererProperties]
  renderers [{rendererClass=&NiagaraSpriteRendererProperties}]
 }
 Light {
  emitterUpdateScript [
   &EmitterState()
   &SpawnBurst_Instantaneous(Spawn Count=1, Use Spawn Probability=false)
  ]
  particleSpawnScript [
   &InitializeParticle(
    Color=<rgba 0 5 30 1>
    Lifetime=0.45
    Sprite Size Mode=::NewEnumerator3
    Uniform Sprite Size=50
    UsePositionOffset=false
    Write Material Random=true
   )
   &SystemLocation(MaskSpawnGroup=false, Offset=<0 -5 20>)
  ]
  particleUpdateScript [
   &ParticleState()
   &ScaleSpriteSize(Scale Factor=&Vector2DFromFloat(Value=&FloatFromCurve(FloatCurve=DataInterface() {propertyValues {Curve {keys [
          {interpMode=RCIM_Cubic, leaveTangent=0.2376, tangentMode=RCTM_Break, time=0, value=0}
          {arriveTangent=0.0486, interpMode=RCIM_Cubic, leaveTangent=0.0486, tangentMode=RCTM_User, time=0.15, value=1}
          {arriveTangent=-2.7369, interpMode=RCIM_Cubic, tangentMode=RCTM_Break, time=1, value=0}
         ]}}})))
   &CameraOffset(Camera Offset Amount=15)
   &SolveForcesAndVelocity(Clamp Velocity=false, Limit Acceleration=false)
   &ScaleColor(
    Scale RGB=&VectorFromFloat(Value=&FloatFromCurve(FloatCurve=DataInterface() {propertyValues {Curve {keys [
          {interpMode=RCIM_Cubic, leaveTangent=0.2376, tangentMode=RCTM_Break, time=0, value=1}
          {arriveTangent=-2.6724, interpMode=RCIM_Cubic, leaveTangent=-2.6724, tangentMode=RCTM_User, time=0.15, value=1}
          {arriveTangent=-0.403, interpMode=RCIM_Cubic, tangentMode=RCTM_Break, time=1, value=0}
         ]}}}))
    ScaleA=true
    ScaleRGB=true
   )
  ]
  rendererClasses [&NiagaraSpriteRendererProperties, &NiagaraLightRendererProperties]
  renderers [
   {rendererClass=&NiagaraSpriteRendererProperties}
   {rendererClass=&NiagaraLightRendererProperties, rendererIndex=1}
  ]
 }
 Ribbons {
  emitterSpawnScript [/SetVariables(Emitter.EmitterRandom:NiagaraInt32=&UniformRangedInt(Minimum=0, Maximum=10000)) {isSetParameters=true}]
  emitterUpdateScript [
   &EmitterState()
   &FN_SpawnRibbon_v2(Number of Ribbons=5, SpawnRate:NiagaraFloat=&UniformRangedFloat(Minimum=400, Maximum=500))
  ]
  particleSpawnScript [
   &FN_InitializeRibbon_v3(
    Lifetime=1
    Ribbon Width:NiagaraFloat=&MakeCustomFloatFromBool(Bool=&SetBoolByFloatComparison(A=$Emitter.LoopedAge, B=0.01), True Float=0, False Float=2)
   )
   &KillParticles(Kill Particles=&SetBoolByFloatComparison(A=$Emitter.LoopedAge, B=0.02))
   /SetVariables(Particles.RandomSeedByRibbon:NiagaraInt32=&Add_Integer(
     A=$Emitter.EmitterRandom
     B=&self:NiagaraRibbonToInt()
    )) {isSetParameters=true}
   /SetVariables(Particles.Lifetime=&FloatFromCurve(
     FloatCurve=DataInterface() {propertyValues {Curve {keys [
         {interpMode=RCIM_Cubic, tangentMode=RCTM_User, time=0, value=1.1}
         {arriveTangent=-0.95, interpMode=RCIM_Cubic, leaveTangent=-0.95, tangentMode=RCTM_User, time=0.25, value=1}
         <at 1 0.05>
        ]}}}
     CurveIndex=$Emitter.LoopedAge
     Scale Curve=&RandomRangeFloat(Minimum=0.75, Maximum=2, Randomness Mode=::NewEnumerator1, Random Seed=$Particles.RandomSeedByRibbon, Fixed Random Seed=true)
    )) {isSetParameters=true}
   /SetVariables(Particles.RibbonDirection:Vector3f=/Niagara/DynamicInputs/Helpers/SafeNormalizeVector(
     Vector To Normalize:Vector3f=&UniformRangedVector(
      Minimum=<0.5 0.1 0.2>
      Maximum=<-0.5 -0.1 0>
      Randomness Mode=::NewEnumerator1
      Random Seed=$Particles.RandomSeedByRibbon
      Fixed Random Seed=true
     )
     Fallback Vector:Vector3f=<0 0 0>
    )) {isSetParameters=true}
   &SphereLocation(
    MaskSpawnGroup=false
    Offset=<0 -5 10>
    Override Seed=false
    Sphere Radius=5
    Surface Only=false
    UseNonUniformScale=false
    UseOffset=true
    UseSphereOrientation=false
   )
   &AddVelocity(Velocity=&Multiply_VectorByFloat(
     Vector=&Add_Vector(A=$Particles.RibbonDirection, B=&RandomRangeVector(Minimum=<0 0 0.1>, Maximum=<0 0 0.2>))
     Float=&Multiply_Float(
      A=&RandomRangeFloat(Minimum=200, Maximum=400)
      B=&FloatFromCurve(FloatCurve=DataInterface() {propertyValues {Curve {keys [<at 0 1>,<at 0.01 0.25>]}}}, CurveIndex=$Emitter.LoopedAge)
     )
    ))
   /SetVariables(Particles.SpriteSize=&Vector2DFromFloat(Value=2)) {isSetParameters=true}
   &Color(Color=&HueShiftLinearColor(
     Hue=&RandomRangeFloat(Minimum=0.5, Maximum=0, Randomness Mode=::NewEnumerator1, Random Seed=$Particles.RandomSeedByRibbon, Fixed Random Seed=true)
     Saturation=&RandomRangeFloat(Minimum=0.75, Maximum=1)
     Value=&RandomRangeFloat(Minimum=0.25, Maximum=2.5)
     Linear Color=<rgba 0.171 0 0 1>
    ))
  ]
  particleUpdateScript [
   &ParticleState(Kill Particles When Lifetime Has Elapsed=false, Let Infinitely Lived Particles Die When Emitter Deactivates=true)
   &Drag(Drag=&Lerp_Float(A=8, B=1, Alpha=$Particles.NormalizedAge), Use Linear Drag=true, Use Rotational Drag=false)
   &GravityForce(Gravity=&RandomRangeVector(Minimum=<0 0 50>, Maximum=<0 0 -100>))
   &AddVelocity(Velocity=&Multiply_VectorByFloat(
     Vector=&RandomRangeVector(Minimum=<0 -20 -20>, Maximum=<1 20 10>, Evaluation Type=::NewEnumerator1)
     Float=&Lerp_Float(A=1, B=0, Alpha=$Particles.NormalizedAge)
    )) {module=AddVelocity001}
   &SolveForcesAndVelocity(Clamp Velocity=false, Limit Acceleration=false)
   /Game/Effects/Niagara/Modules/Contrails/FN_UpdateRibbon/FN_UpdateRibbon_v3()
   &ScaleRibbonWidth(Ribbon Width Scale=&FloatFromCurve(
     FloatCurve=DataInterface() {propertyValues {Curve {keys [
         {arriveTangent=5.9146, interpMode=RCIM_Cubic, leaveTangent=5.9146, tangentMode=RCTM_User, time=0, value=0}
         <at 0.495 1 RCIM_Cubic>
         {arriveTangent=-2.495, interpMode=RCIM_Cubic, leaveTangent=-2.495, tangentMode=RCTM_User, time=0.7, value=0}
        ]}}}
     Scale Curve=&RandomRangeFloat(Minimum=0.4, Maximum=1, Random Seed=$Particles.RandomSeedByRibbon, Fixed Random Seed=true)
    ))
   &DynamicMaterialParameters(
    Index 0 Param 4=1
    Module.Param0WriteEnabled=true
    Module.Param1WriteEnabled=true
    Module.Param2WriteEnabled=true
    Module.Param3WriteEnabled=true
   )
   &KillParticles(Kill Particles=&SetBoolByFloatComparison(A=$Emitter.LoopedAge, B=0.7)) {module=KillParticles001}
  ]
  rendererClasses [&NiagaraRibbonRendererProperties]
  renderers [{rendererClass=&NiagaraRibbonRendererProperties}]
 }
 Smoke {
  emitterUpdateScript [
   &EmitterState()
   &SpawnBurst_Instantaneous(Spawn Count=10, Use Spawn Probability=false)
  ]
  particleSpawnScript [
   /Niagara/Modules/Spawn/Initialization/InitializeParticle(
    Color=&RandomRangeLinearColor(Minimum=<rgba 0.4688 0.4688 0.4688 0.8>, Maximum=<rgba 0.4688 0.4688 0.4688 0.9>)
    Lifetime=&RandomRangeFloat(Minimum=1, Maximum=1.5)
    Module.Write Color=true
    Module.Write Lifetime=true
    Module.Write Mass=true
    Module.Write Material Random=true
    Module.Write Position=true
    Module.Write Scale=false
    Module.Write SpriteRotation=true
    Module.Write SpriteSize=true
    Sprite Rotation=&RandomRangeFloat(Minimum=0, Maximum=360)
    Sprite Size=&Vector2DFromFloat(Value=&RandomRangeFloat(Minimum=35, Maximum=45))
   )
   &SphereLocation(
    MaskSpawnGroup=false
    Offset=<0 -5 8>
    Override Seed=false
    Sphere Radius=2
    Surface Only=false
    UseNonUniformScale=false
    UseOffset=true
    UseSphereOrientation=false
   )
   &AddVelocityFromPoint(Module.Use Falloff Distance=false, Velocity Strength=&RandomRangeFloat(Minimum=10, Maximum=60))
   &FN_SetSpawnGroup(Number of Spawn Groups=2)
   &self:Add_Velocity_From_Spawn_Group(
    Velocity:Vector3f=&RandomRangeVector(Minimum=<100 0 80>, Maximum=<30 0 60>)
    Velocity Second:Vector3f=&RandomRangeVector(Minimum=<-100 0 80>, Maximum=<-30 0 60>)
   ) {module=Add_Velocity_From_Spawn_Group}
  ]
  particleUpdateScript [
   &ParticleState()
   &Drag(Drag=&RandomRangeFloat(Minimum=3.5, Maximum=3), Use Linear Drag=true, Use Rotational Drag=false)
   &GravityForce(Gravity=&RandomRangeVector(Minimum=<0 0 -100>, Maximum=<0 0 -200>))
   &SolveForcesAndVelocity(Clamp Velocity=false, Limit Acceleration=false)
   &SubUVAnimation(
    End Frame=50
    Start Frame=10
    SubUV Animation Mode:ENiagara_SubUVLookupMode=::NewEnumerator1
    SubUV Curve=DataInterface() {propertyValues {Curve {keys [
        {arriveTangent=1.5272, interpMode=RCIM_Cubic, leaveTangent=1.5272, tangentMode=RCTM_User, time=0, value=0}
        {arriveTangent=0.0067, interpMode=RCIM_Cubic, leaveTangent=0.0067, tangentMode=RCTM_User, time=1, value=1}
       ]}}}
   )
   &ScaleSpriteSize(Scale Factor=&Vector2DFromFloat(Value=&FloatFromCurve(FloatCurve=DataInterface() {propertyValues {Curve {keys [
          {arriveTangent=0.0059, interpMode=RCIM_Cubic, leaveTangent=0.0059, tangentMode=RCTM_User, time=0, value=1}
          {arriveTangent=0.1826, interpMode=RCIM_Cubic, leaveTangent=0.1826, tangentMode=RCTM_User, time=1, value=1.8}
         ]}}})))
   &ScaleColor(
    Linear Color Curve=DataInterface() {propertyValues {
      AlphaCurve {keys [<at 0 1>,<at 1 0>]}
      BlueCurve {keys [<at 0.0013 1>,<at 0.435 1>,<at 1 1>]}
      GreenCurve {keys [<at 0.0013 0.0833>,<at 0.435 0>,<at 1 0>]}
      RedCurve {keys [<at 0.0013 0>,<at 0.435 0.1667>,<at 1 0.4167>]}
     }}
    Scale Mode=::NewEnumerator2
   ) {module=ScaleColor001}
   &ScaleColor(
    Scale Alpha=&FloatFromCurve(FloatCurve=DataInterface() {propertyValues {Curve {keys [
         {interpMode=RCIM_Cubic, tangentMode=RCTM_User, time=0, value=1.2}
         {arriveTangent=-1.7719, interpMode=RCIM_Cubic, leaveTangent=-1.7719, tangentMode=RCTM_User, time=0.8, value=0.75}
         {arriveTangent=-4.4359, interpMode=RCIM_Cubic, tangentMode=RCTM_Break, time=1, value=0}
        ]}}})
    Scale RGB=&VectorFromFloat(Value=&FloatFromCurve(FloatCurve=DataInterface() {propertyValues {Curve {keys [
          {arriveTangent=-1.4923, interpMode=RCIM_Cubic, leaveTangent=-1.4923, tangentMode=RCTM_User, time=0, value=1.5}
          {arriveTangent=-1.2406, interpMode=RCIM_Cubic, leaveTangent=-1.2406, tangentMode=RCTM_User, time=0.9, value=0.5}
          {arriveTangent=-1.9627, interpMode=RCIM_Cubic, leaveTangent=-1.9627, tangentMode=RCTM_User, time=1, value=0}
         ]}}}))
    ScaleA=true
    ScaleRGB=true
   )
   &CameraOffset(Camera Offset Amount=&RandomRangeFloat(Minimum=0, Maximum=5))
  ]
  rendererClasses [&NiagaraSpriteRendererProperties]
  renderers [{rendererClass=&NiagaraSpriteRendererProperties}]
 }
 Smoke_Light {
  emitterUpdateScript [
   &EmitterState()
   &SpawnBurst_Instantaneous(Spawn Count=4, Use Spawn Probability=false)
  ]
  particleSpawnScript [
   &InitializeParticle(
    Color=&RandomRangeLinearColor(Minimum=<rgba 0 0 0.4688 1>, Maximum=<rgba 0 0.0781 0.4688 1>)
    Lifetime=&RandomRangeFloat(Minimum=1.2, Maximum=1.5)
    Module.Write Color=true
    Module.Write Lifetime=true
    Module.Write Mass=true
    Module.Write Material Random=true
    Module.Write Position=true
    Module.Write Scale=false
    Module.Write SpriteRotation=true
    Module.Write SpriteSize=true
    Sprite Rotation=&RandomRangeFloat(Minimum=0, Maximum=360)
    Sprite Size=&Vector2DFromFloat(Value=&RandomRangeFloat(Minimum=25, Maximum=45))
   )
   &SphereLocation(
    MaskSpawnGroup=false
    Offset=<0 -5 8>
    Override Seed=false
    Sphere Radius=4
    Surface Only=false
    UseNonUniformScale=false
    UseOffset=true
    UseSphereOrientation=false
   )
   &AddVelocityFromPoint(Module.Use Falloff Distance=false, Velocity Strength=&RandomRangeFloat(Minimum=10, Maximum=60))
   &FN_SetSpawnGroup(Number of Spawn Groups=2)
   &self:Add_Velocity_From_Spawn_Group(
    Velocity:Vector3f=&RandomRangeVector(Minimum=<100 0 40>, Maximum=<30 0 60>)
    Velocity Second:Vector3f=&RandomRangeVector(Minimum=<-100 0 40>, Maximum=<-30 0 60>)
   ) {module=Add_Velocity_From_Spawn_Group}
  ]
  particleUpdateScript [
   &ParticleState()
   &Drag(Drag=&RandomRangeFloat(Minimum=4.5, Maximum=5), Use Linear Drag=true, Use Rotational Drag=false)
   &GravityForce(Gravity=&RandomRangeVector(Minimum=<0 0 -100>, Maximum=<0 0 -70>))
   &SolveForcesAndVelocity(Clamp Velocity=false, Limit Acceleration=false)
   &SubUVAnimation(
    End Frame=50
    Start Frame=20
    SubUV Animation Mode:ENiagara_SubUVLookupMode=::NewEnumerator1
    SubUV Curve=DataInterface() {propertyValues {Curve {keys [
        {arriveTangent=1.5272, interpMode=RCIM_Cubic, leaveTangent=1.5272, tangentMode=RCTM_User, time=0, value=0}
        {arriveTangent=0.0067, interpMode=RCIM_Cubic, leaveTangent=0.0067, tangentMode=RCTM_User, time=1, value=1}
       ]}}}
   )
   &ScaleSpriteSize(Scale Factor=&Vector2DFromFloat(Value=&FloatFromCurve(FloatCurve=DataInterface() {propertyValues {Curve {keys [
          {arriveTangent=0.0059, interpMode=RCIM_Cubic, leaveTangent=0.0059, tangentMode=RCTM_User, time=0, value=1}
          {arriveTangent=0.1826, interpMode=RCIM_Cubic, leaveTangent=0.1826, tangentMode=RCTM_User, time=1, value=1.8}
         ]}}})))
   &ScaleColor(
    Scale Alpha=&FloatFromCurve(FloatCurve=DataInterface() {propertyValues {Curve {keys [
         {interpMode=RCIM_Cubic, tangentMode=RCTM_User, time=0, value=1.2}
         {arriveTangent=-1.1646, interpMode=RCIM_Cubic, tangentMode=RCTM_Break, time=1, value=0}
        ]}}})
    Scale RGB=&VectorFromFloat(Value=&FloatFromCurve(FloatCurve=DataInterface() {propertyValues {Curve {keys [
          {arriveTangent=-3.6144, interpMode=RCIM_Cubic, leaveTangent=-3.6144, tangentMode=RCTM_User, time=0.05, value=5}
          {arriveTangent=-1.1838, interpMode=RCIM_Cubic, leaveTangent=-1.1838, tangentMode=RCTM_User, time=0.2, value=1}
          {interpMode=RCIM_Cubic, tangentMode=RCTM_User, time=1, value=0.5}
         ]}}}))
    ScaleA=true
    ScaleRGB=true
   )
   &CameraOffset(Camera Offset Amount=&RandomRangeFloat(Minimum=5, Maximum=10))
  ]
  rendererClasses [&NiagaraSpriteRendererProperties]
  renderers [{rendererClass=&NiagaraSpriteRendererProperties}]
 }
 Sparkles {
  simTarget=GPUComputeSim
  emitterUpdateScript [
   &EmitterState()
   &SpawnBurst_Instantaneous(Spawn Count=100, Use Spawn Probability=false)
  ]
  particleSpawnScript [
   &InitializeParticle(
    Color=&HueShiftLinearColor(
     Hue=0
     Saturation=&RandomRangeFloat(Minimum=5, Maximum=10)
     Value=&RandomRangeFloat(Minimum=1, Maximum=5)
     Linear Color=&RandomRangeLinearColor(Minimum=<rgba 0.8 0 0.8 1>, Maximum=<rgba 0.1333 0 0.8 1>, Color Channel Mode=::NewEnumerator2)
    )
    Lifetime=&Power(a=&RandomRangeFloat(Minimum=0.46, Maximum=1), b=2)
    Module.Write Color=true
    Module.Write Lifetime=true
    Module.Write Mass=true
    Module.Write Material Random=true
    Module.Write Position=false
    Module.Write Scale=false
    Module.Write SpriteRotation=true
    Module.Write SpriteSize=true
    Position=&SimulationPosition()
    Sprite Rotation=&RandomRangeFloat(Minimum=0, Maximum=360)
    Sprite Size=&Vector2DFromFloat(Value=&RandomRangeFloat(Minimum=1, Maximum=5))
   )
   &SphereLocation(
    MaskSpawnGroup=false
    Override Seed=false
    Sphere Radius=&RandomRangeFloat(Minimum=2, Maximum=10)
    Surface Only=false
    UseNonUniformScale=false
    UseOffset=false
    UseSphereOrientation=false
   )
   &AddVelocityFromPoint(Module.Use Falloff Distance=false, Velocity Strength=&RandomRangeFloat(Minimum=50, Maximum=150))
   &FN_SetSpawnGroup(Number of Spawn Groups=2)
   &self:Add_Velocity_From_Spawn_Group(Velocity:Vector3f=<100 0 200>, Velocity Second:Vector3f=<-100 0 200>) {module=Add_Velocity_From_Spawn_Group}
  ]
  particleUpdateScript [
   &ParticleState()
   &Drag(Drag=5, Use Linear Drag=true, Use Rotational Drag=false)
   &GravityForce(Gravity=&RandomRangeVector(Minimum=<0 0 -200>, Maximum=<1 1 -500>))
   &CurlNoiseForce(Noise Strength=200, Pan Noise=false)
   &SolveForcesAndVelocity(Clamp Velocity=false, Limit Acceleration=false)
   &ScaleColor(
    Scale Alpha=5
    Scale RGB=&VectorFromFloat(Value=&RandomRangeFloat(Minimum=0.04, Maximum=0.01))
    ScaleA=true
    ScaleRGB=true
   )
   &ScaleSpriteSize(Scale Factor=&Vector2DFromFloat(Value=&FloatFromCurve(FloatCurve=DataInterface() {propertyValues {Curve {keys [<at 0.0018 0.1957>,<at 0.0527 0.819>,<at 0.1169 0.6555>,<at 0.1748 0.9015>,<at 0.2394 0.582>,<at 0.332 1.0023>,<at 0.4947 0.3707>,<at 0.6727 1>,<at 1 0>]}}})))
  ]
  rendererClasses [&NiagaraSpriteRendererProperties]
  renderers [{rendererClass=&NiagaraSpriteRendererProperties}]
 }
 Strips {
  emitterUpdateScript [
   &EmitterState()
   &SpawnBurst_Instantaneous(Spawn Count=12, Use Spawn Probability=false)
  ]
  particleSpawnScript [
   &InitializeParticle(
    Color=&MakeCustomLinearColorFromBool(
     Bool=&RandomBool()
     True Color=&RandomRangeLinearColor(Minimum=<rgba 0 0.305 0 1>, Maximum=<rgba 0.4 0.305 0 1>)
     False Color=&RandomRangeLinearColor(Minimum=<rgba 0.85 0.305 0 1>, Maximum=<rgba 1 0.305 0 1>)
    )
    Lifetime=&RandomRangeFloat(Minimum=0.4, Maximum=0.8)
    Module.Write Color=true
    Module.Write Lifetime=true
    Module.Write Mass=true
    Module.Write Material Random=true
    Module.Write Position=false
    Module.Write Scale=false
    Module.Write SpriteRotation=true
    Module.Write SpriteSize=true
    Position=&SimulationPosition()
    Sprite Rotation=&RandomRangeFloat(Minimum=0, Maximum=360)
    Sprite Size=&Vector2DFromFloat(Value=&RandomRangeFloat(Minimum=5, Maximum=12))
   )
   &SphereLocation(
    MaskSpawnGroup=false
    Offset=<0 -5 8>
    Override Seed=false
    Sphere Radius=&RandomRangeFloat(Minimum=5, Maximum=15)
    Surface Only=false
    UseNonUniformScale=false
    UseOffset=true
    UseSphereOrientation=false
   )
   &AddVelocityFromPoint(Module.Use Falloff Distance=false, Velocity Strength=&RandomRangeFloat(Minimum=10, Maximum=120))
   &FN_SetSpawnGroup(Number of Spawn Groups=2)
   &self:Add_Velocity_From_Spawn_Group(Velocity:Vector3f=<100 0 100>, Velocity Second:Vector3f=<-100 0 100>) {module=Add_Velocity_From_Spawn_Group}
  ]
  particleUpdateScript [
   &ParticleState()
   &Drag(Drag=4, Use Linear Drag=true, Use Rotational Drag=false)
   &GravityForce(Gravity=&RandomRangeVector(Minimum=<0 0 -360>, Maximum=<1 1 -200>))
   &SolveForcesAndVelocity(Clamp Velocity=false, Limit Acceleration=false)
   &ScaleSpriteSize(Scale Factor=&Vector2DFromFloat(Value=&FloatFromCurve(FloatCurve=DataInterface() {propertyValues {Curve {keys [
          {arriveTangent=7.4484, interpMode=RCIM_Cubic, leaveTangent=7.4484, tangentMode=RCTM_User, time=0, value=0}
          {arriveTangent=0.4579, interpMode=RCIM_Cubic, leaveTangent=0.4579, tangentMode=RCTM_User, time=0.1, value=1}
          {arriveTangent=-0.4596, interpMode=RCIM_Cubic, leaveTangent=-0.4596, tangentMode=RCTM_User, time=0.55, value=1}
          {arriveTangent=-4.0966, interpMode=RCIM_Cubic, leaveTangent=-4.0966, tangentMode=RCTM_User, time=1, value=0}
         ]}}})))
  ]
  rendererClasses [&NiagaraSpriteRendererProperties]
  renderers [{rendererClass=&NiagaraSpriteRendererProperties}]
 }
}
summary {
 systemName=NS_Backpack_F_MED_Cotton_Candy_Surprise
 emitters [
  {emitterName=Glow, rendererClasses [&NiagaraSpriteRendererProperties]}
  {emitterName=Fragments, rendererClasses [&NiagaraSpriteRendererProperties]}
  {emitterName=Strips, rendererClasses [&NiagaraSpriteRendererProperties]}
  {emitterName=Ribbons, rendererClasses [&NiagaraRibbonRendererProperties]}
  {emitterName=Sparkles, simTarget=GPUComputeSim, rendererClasses [&NiagaraSpriteRendererProperties]}
  {emitterName=Smoke, rendererClasses [&NiagaraSpriteRendererProperties]}
  {emitterName=FireworkParent, rendererClasses [&NiagaraSpriteRendererProperties]}
  {emitterName=FireworkChild, rendererClasses [&NiagaraSpriteRendererProperties]}
  {emitterName=Smoke_Light, rendererClasses [&NiagaraSpriteRendererProperties]}
  {emitterName=Light, rendererClasses [&NiagaraSpriteRendererProperties, &NiagaraLightRendererProperties]}
  {emitterName=FlashStars, rendererClasses [&NiagaraSpriteRendererProperties]}
 ]
 userVariables [User.RibbonIDOffset {type=NiagaraInt32}]
}
