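# Material function for the Cotton Candy backpack that adds a glowing digital countdown of the safe zone phase time to a material's emissive.
# UV1 is split into six digit columns, each picking a flipbook row from the collection parameter's hours, minutes and seconds plus offsets.
# Sine-pulsed separators and a faint background "8" join the tinted digits.
# Explicit derivatives from scaled UV0 drive mip selection across the discontinuous digit UVs.
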MaterialFunction /BRCosmetics/Accessories/FORT_Backpacks/Backpack_F_MED_Cotton_Candy/Materials/MF_Cotton_Candy
graph {nodes [
  n1 = GetMaterialAttributes(None=FunctionInput() {inputName=Material, inputType=FunctionInput_MaterialAttributes}) {attributeGetTypes [#1]}
  n2 = Sine(None=Multiply(A=Time(), B=ScalarParameter() {defaultValue=1, parameterName=CC_Tic_Speed}))
  n3 = TextureCoordinate() {coordinateIndex=1}
  n4 = ComponentMask(None=@n3) {r=true}
  n5 = Add(A=@n4) {constB=-1}
  n6 = Divide() {constA=2.5, constB=6}
  n7 = Saturate(None=Divide(A=Add(A=@n5, B=@n6)) {constB=0.001})
  n8 = Add(A=@n4) {constB=-1}
  n9 = Divide() {constA=1.5, constB=6}
  n10 = Saturate(None=Divide(A=Add(A=@n8, B=@n9)) {constB=0.001})
  n11 = ComponentMask(None=@n3) {r=true}
  n12 = Add(A=@n11) {constB=-1}
  n13 = Divide() {constA=4.5, constB=6}
  n14 = Saturate(None=Divide(A=Add(A=@n12, B=@n13)) {constB=0.001})
  n15 = Add(A=@n11) {constB=-1}
  n16 = Divide() {constA=3.5, constB=6}
  n17 = Saturate(None=Divide(A=Add(A=@n15, B=@n16)) {constB=0.001})
  n18 = Add(A=Multiply(A=ComponentMask(None=@n3) {r=true}, B=Constant() {r=5.8}) {constB=8}) {constB=-0.44}
  n19 = Divide(B=ScalarParameter() {defaultValue=10, group=CottonCandy, parameterName=CC_No_Columns}) {constB=10}
  n20 = Multiply(A=ComponentMask(None=@n3) {g=true}, B=@n19)
  n21 = TextureObjectParameter() {
   group=CottonCandy
   parameterName=CC_Font_Text
   texture=/Game/Accessories/FORT_Backpacks/Backpack_F_MED_Cotton_Candy/Textures/T_NumberFlipBook
  }
  n22 = Multiply(A=Add(A=Subtract(A=@n7, B=@n10), B=Subtract(A=@n14, B=@n17)), B=TextureSample(UVs=AppendVector(A=@n18, B=@n20), Tex=@n21).G)
  n23 = Multiply(A=Multiply(A=Add(A=@n2)) {constB=0.5}, B=@n22)
  n24 = ComponentMask(None=@n3) {r=true}
  n25 = Constant() {r=6}
  n26 = ComponentMask(None=@n3) {r=true}
  n27 = Add(A=@n26) {constB=-1}
  n28 = Divide() {constA=2, constB=6}
  n29 = Saturate(None=Divide(A=Add(A=@n27, B=@n28)) {constB=0.001})
  n30 = Multiply(A=@n29, B=Multiply(A=@n29, B=Constant() {r=-0.1}))
  n31 = Add(A=Add(A=@n26) {constB=-1}, B=Divide() {constA=6, constB=6})
  n32 = Add(A=Add(A=@n26) {constB=-1}, B=Divide() {constA=4, constB=6})
  n33 = Subtract(A=Saturate(None=Divide(A=@n31) {constB=0.001}), B=Saturate(None=Divide(A=@n32) {constB=0.001}))
  n34 = Multiply(A=@n33, B=Multiply(A=@n33, B=Constant() {r=0.1}))
  n35 = Frac(None=Add(A=Multiply(A=@n24, B=@n25) {constB=8}, B=Add(A=@n30, B=@n34)))
  n36 = CollectionParameter() {
   collection=&FortniteMaterialParameters
   parameterName=AthenaSafeZoneGamePhaseStepTimeRemaining
  }
  n37 = MaterialFunctionCall(Float4=@n36) {materialFunction=&BreakOutFloat4Components}
  n38 = Add(A=@n37.R, B=ScalarParameter() {group=CottonCandy, parameterName=CC_Number_Offset_Hours})
  n39 = Divide(A=@n38) {constB=10}
  n40 = Subtract(A=@n39, B=Frac(None=@n39))
  n41 = Subtract(A=@n38, B=Frac(None=@n38))
  n42 = ComponentMask(None=@n26) {r=true}
  n43 = Add(A=@n42) {constB=-1}
  n44 = Divide() {constA=5, constB=6}
  n45 = Saturate(None=Divide(A=Add(A=@n43, B=@n44)) {constB=0.001})
  n46 = Add(A=@n37.G, B=ScalarParameter() {group=CottonCandy, parameterName=CC_Number_Offset_Minutes})
  n47 = Divide(A=@n46) {constB=10}
  n48 = Frac(None=@n47)
  n49 = Divide(A=Add(A=Add(A=@n42) {constB=-1}, B=Divide() {constA=4, constB=6})) {constB=0.001}
  n50 = Add(A=Add(A=@n42) {constB=-1}, B=Divide() {constA=3, constB=6})
  n51 = LinearInterpolate(
   A=LinearInterpolate(A=LinearInterpolate(A=@n40, B=@n41, Alpha=@n45), B=Subtract(A=@n47, B=@n48), Alpha=Saturate(None=@n49))
   B=Subtract(A=@n46, B=Frac(None=@n46))
   Alpha=Saturate(None=Divide(A=@n50) {constB=0.001})
  )
  n52 = Add(A=@n37.B, B=ScalarParameter() {group=CottonCandy, parameterName=CC_Number_Offset_Seconds})
  n53 = Divide(A=@n52) {constB=10}
  n54 = Subtract(A=@n53, B=Frac(None=@n53))
  n55 = Add(A=@n42) {constB=-1}
  n56 = Divide() {constA=2, constB=6}
  n57 = Saturate(None=Divide(A=Add(A=@n55, B=@n56)) {constB=0.001})
  n58 = Constant() {r=10}
  n59 = Divide(A=Add(A=Add(A=@n42) {constB=-1}, B=Divide() {constB=6})) {constB=0.001}
  n60 = Multiply(A=@n19, B=LinearInterpolate(A=LinearInterpolate(A=@n51, B=@n54, Alpha=@n57) {constB=0}, B=Fmod(A=@n52, B=@n58), Alpha=Saturate(None=@n59)))
  n61 = Multiply(A=TextureCoordinate()) {constB=3}
  n62 = DDX(Value=@n61)
  n63 = DDY(Value=@n61)
  n64 = TextureSample!(UVs=AppendVector(A=@n35, B=Add(A=@n20, B=@n60) {constB=0}), Tex=@n21, "DDX(UVs)"=@n62, "DDY(UVs)"=@n63, Apply View MipBias) {mipValueMode=TMVM_Derivative}
  n65 = Multiply(A=@n19) {constB=8}
  n66 = TextureSample(UVs=AppendVector(A=@n35, B=Add(A=@n20, B=@n65) {constB=2}), Tex=@n21, "DDX(UVs)"=@n62, "DDY(UVs)"=@n63) {mipValueMode=TMVM_Derivative}
  n67 = Multiply(
   A=Add(A=Add(A=@n23, B=@n64.R), B=Multiply(A=@n66.R) {constB=0.05})
   B=VectorParameter() {parameterName=CC_Font_Color, defaultValue=<rgba 0.651 0 0 1>}.RGB
  )
  FunctionOutput(None=SetMaterialAttributes!(MaterialAttributes=@n1.MaterialAttributes, EmissiveColor=Add(A=@n1.EmissiveColor, B=@n67)) {attributeSetTypes [#1]}) {outputName=Result_1}
  LinearInterpolate(A=@n64.R, B=@n22)
  MaterialFunctionCall(Color=@n36) {materialFunction=&SplitComponents}
 ]}
