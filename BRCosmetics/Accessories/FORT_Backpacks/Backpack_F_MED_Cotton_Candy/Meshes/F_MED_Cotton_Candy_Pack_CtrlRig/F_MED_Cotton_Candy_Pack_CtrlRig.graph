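# Control Rig for the Cotton Candy backpack cosmetic, procedurally posing its spring, head, grenades and side boxes.
# The local X scale of dyn_spring_MSTR is the single driver, fed through clamped remaps into bone translation, rotation and scale.
# The spring is modified first, then re-read for the head, grenade and box steps.
# Head rotation passes through an accumulated quaternion lerp; several leftover remap and rotation nodes are unconnected.
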ControlRigBlueprint /BRCosmetics/Accessories/FORT_Backpacks/Backpack_F_MED_Cotton_Candy/Meshes/F_MED_Cotton_Candy_Pack_CtrlRig/F_MED_Cotton_Candy_Pack_CtrlRig
graphs [
 RigVMFunctionLibrary {links [], nodes []}
 RigVMModel {nodes [
   RigUnit_BeginExecution = &RigUnit_BeginExecution()
   n1 = &RigUnit_GetTransform(Item=(Type=Bone,Name="dyn_spring_MSTR"), Space=LocalSpace)
   n2 = &RigVMFunction_MathFloatRemap(SourceMaximum=1.1, SourceMinimum=0.8, TargetMaximum=0.7, TargetMinimum=1.3, bClamp=true, Value=@n1.Transform.Scale3D.X)
   &RigUnit_ModifyTransforms(
    ItemToModify="((Item=(Type=Bone,Name=\"dyn_spring_MSTR\"),Transform=(Rotation=(X=0,Y=0,Z=0,W=1),Translation=(X=0,Y=0,Z=0),Scale3D=(X=1,Y=1,Z=1))),(Item=(Type=Bone,Name=\"dyn_spring_MSTR\"),Transform=(Rotation=(X=0,Y=0,Z=0,W=1),Translation=(X=0,Y=0,Z=0),Scale3D=(X=1,Y=1,Z=1) ...(truncated)"
    Weight=1
    ItemToModify.0.Transform.Translation.X=&RigVMFunction_MathFloatRemap(SourceMaximum=0.8, SourceMinimum=0.2, TargetMaximum=2.5, TargetMinimum=0, bClamp=true, Value=@n1.Transform.Scale3D.X).Result
    ItemToModify.0.Transform.Scale3D.X=&RigVMFunction_MathFloatRemap(SourceMaximum=0.8, SourceMinimum=0.2, TargetMaximum=1.3, TargetMinimum=0.9, bClamp=true, Value=@n1.Transform.Scale3D.X).Result
    ItemToModify.1.Transform.Scale3D.X=@n2.Result
    ItemToModify.1.Transform.Scale3D.Y=@n2.Result
    ItemToModify.1.Transform.Scale3D.Z=@n2.Result
    ItemToModify.1.Transform.Translation.X=&RigVMFunction_MathFloatRemap(SourceMaximum=1.1, SourceMinimum=0.9, TargetMaximum=-2, TargetMinimum=0, bClamp=true, Value=@n1.Transform.Scale3D.X).Result
    ExecutePin=@RigUnit_BeginExecution.ExecutePin
   )
   n3 = &RigUnit_GetTransform(Item=(Type=Bone,Name="dyn_spring_MSTR"), Space=LocalSpace)
   n4 = &RigVMFunction_MathFloatRemap(SourceMaximum=0.6, SourceMinimum=0.2, TargetMaximum=-40, TargetMinimum=40, bClamp=true, Value=@n3.Transform.Scale3D.X)
   n5 = &RigVMFunction_MathFloatRemap(SourceMaximum=0.6, SourceMinimum=0.2, TargetMaximum=1.5, TargetMinimum=0.8, bClamp=true, Value=@n3.Transform.Scale3D.X)
   > &RigUnit_ModifyTransforms(
    ItemToModify=((Item=(Type=Bone,Name="dyn_head"),Transform=(Rotation=(X=0,Y=0,Z=0,W=1),Translation=(X=0,Y=0,Z=0),Scale3D=(X=1,Y=1,Z=1))),(Item=(Type=Bone,Name="dyn_head"),Transform=(Rotation=(X=0,Y=0,Z=0,W=1),Translation=(X=0,Y=0,Z=0),Scale3D=(X=1,Y=1,Z=1))))
    Weight=1
    ItemToModify.0.Transform.Rotation=/Script/RigVM.RigVMFunction_AccumulateQuatLerp(
     Blend=1
     bIntegrateDeltaTime=false
     InitialValue=&RigVMFunction_MathQuaternionFromEuler(Euler=(X=0,Y=0,Z=0), Euler.Y=@n3.Transform.Scale3D.X).Result
     TargetValue=&RigVMFunction_MathQuaternionFromEuler(Euler=(X=0,Y=0,Z=0), Euler.Y=@n4.Result).Result
    ) {is=struct RigVMUnitNode, template="AccumulateLerp::Execute(in TargetValue,in InitialValue,in Blend,in bIntegrateDeltaTime,out Result)"}.Result
    ItemToModify.0.Transform.Scale3D.X=@n5.Result
    ItemToModify.0.Transform.Scale3D.Z=@n5.Result
   )
   n6 = &RigUnit_GetTransform(Item=(Type=Bone,Name="dyn_spring_MSTR"), Space=LocalSpace)
   n7 = &RigVMFunction_MathFloatRemap(SourceMaximum=1, SourceMinimum=0.4, TargetMaximum=1.3, TargetMinimum=0.5, bClamp=true, Value=@n6.Transform.Scale3D.X)
   n8 = &RigVMFunction_MathFloatRemap(SourceMaximum=1, SourceMinimum=0.4, TargetMaximum=60, TargetMinimum=-30, bClamp=true, Value=@n6.Transform.Scale3D.X)
   n9 = &RigVMFunction_MathFloatRemap(SourceMaximum=1, SourceMinimum=0.4, TargetMaximum=-60, TargetMinimum=30, bClamp=true, Value=@n6.Transform.Scale3D.X)
   > &RigUnit_ModifyTransforms(
    ItemToModify="((Item=(Type=Bone,Name=\"dyn_l_grenade_MSTR\"),Transform=(Rotation=(X=0,Y=0,Z=0,W=1),Translation=(X=0,Y=0,Z=0),Scale3D=(X=1,Y=1,Z=1))),(Item=(Type=Bone,Name=\"dyn_r_grenade_MSTR\"),Transform=(Rotation=(X=0,Y=0,Z=0,W=1),Translation=(X=0,Y=0,Z=0),Scale3D=(X=1,Y= ...(truncated)"
    Weight=1
    ItemToModify.0.Transform.Scale3D.X=@n7.Result
    ItemToModify.0.Transform.Scale3D.Y=@n7.Result
    ItemToModify.0.Transform.Scale3D.Z=@n7.Result
    ItemToModify.1.Transform.Scale3D.X=@n7.Result
    ItemToModify.1.Transform.Scale3D.Y=@n7.Result
    ItemToModify.1.Transform.Scale3D.Z=@n7.Result
    ItemToModify.0.Transform.Rotation=&RigVMFunction_MathQuaternionFromEuler(Euler=(X=0,Y=0,Z=0), Euler.X=@n8.Result, Euler.Z=@n8.Result).Result
    ItemToModify.1.Transform.Rotation=&RigVMFunction_MathQuaternionFromEuler(Euler=(X=0,Y=0,Z=0), Euler.X=@n9.Result, Euler.Y=@n9.Result).Result
   )
   n10 = &RigUnit_GetTransform(Item=(Type=Bone,Name="dyn_spring_MSTR"), Space=LocalSpace, bInitial=false)
   > &RigUnit_ModifyTransforms(
    ItemToModify="((Item=(Type=Bone,Name=\"dyn_box_l_MSTR\"),Transform=(Rotation=(X=0,Y=0,Z=0,W=1),Translation=(X=-0.157,Y=0,Z=0),Scale3D=(X=1,Y=0.965,Z=1))),(Item=(Type=Bone,Name=\"dyn_box_r_MSTR\"),Transform=(Rotation=(X=0,Y=0,Z=0,W=1),Translation=(X=-0.157,Y=0,Z=0),Scale3D=( ...(truncated)"
    Weight=1
    ItemToModify.0.Transform.Rotation=&RigVMFunction_MathQuaternionFromEuler(
     Euler=(X=0,Y=0,Z=0)
     Euler.Y=&RigVMFunction_MathFloatRemap(SourceMaximum=1, SourceMinimum=0.234, TargetMaximum=0, TargetMinimum=-210, bClamp=true, Value=@n10.Transform.Scale3D.X).Result
    ).Result
    ItemToModify.1.Transform.Rotation=&RigVMFunction_MathQuaternionFromEuler(
     Euler=(X=0,Y=0,Z=0)
     Euler.Y=&RigVMFunction_MathFloatRemap(SourceMaximum=1, SourceMinimum=0.234, TargetMaximum=0, TargetMinimum=-210, bClamp=true, Value=@n10.Transform.Scale3D.X).Result
    ).Result
   )
   &RigVMFunction_AccumulateQuatLerp(
    Blend=1
    bIntegrateDeltaTime=false
    InitialValue=&RigVMFunction_MathQuaternionFromEuler(Euler=(X=0,Y=0,Z=0), Euler.Y=@n3.Transform.Scale3D.X).Result
    TargetValue=&RigVMFunction_MathQuaternionFromEuler(
     Euler=(X=0,Y=0,Z=0)
     Euler.Y=&RigVMFunction_MathFloatRemap(SourceMaximum=1.1, SourceMinimum=0.6, TargetMaximum=30, TargetMinimum=0, bClamp=true, Value=@n3.Transform.Scale3D.X).Result
    ).Result
   ) {template="AccumulateLerp::Execute(in TargetValue,in InitialValue,in Blend,in bIntegrateDeltaTime,out Result)"}
   &RigVMFunction_MathFloatRemap(SourceMaximum=1, SourceMinimum=0.7, TargetMaximum=1.3, TargetMinimum=0.5, Value=0, bClamp=true)
   &RigVMFunction_MathFloatRemap(SourceMaximum=1, SourceMinimum=0.8, TargetMaximum=0.3, TargetMinimum=1, bClamp=true, Value=@n3.Transform.Scale3D.X)
   &RigVMFunction_MathQuaternionFromEuler(Euler=(X=0,Y=0,Z=0))
   &RigVMFunction_MathQuaternionFromEuler(
    Euler=(X=0,Y=0,Z=0)
    Euler.Y=&RigVMFunction_MathFloatRemap(SourceMaximum=1, SourceMinimum=0.4, TargetMaximum=-50, TargetMinimum=50, Value=0, bClamp=true).Result
   )
  ]}
]
hierarchy [
 root {loc=<0 0 0>, rot=<0 0 0 1>}
 pack {parent=root, loc=<0 -0.7316 0>, rot=<0 0 0 1>}
 dyn_spring_MSTR {parent=pack, loc=<0 -7.2371 -4.7623>, rot=<0.5 0.5 0.5 -0.5>}
 dyn_spring_01 {parent=dyn_spring_MSTR, loc=<0.7623 0 0>, rot=<0 0 0 1>}
 dyn_spring_02 {parent=dyn_spring_01, loc=<4 0 0>, rot=<0 0 0 1>}
 dyn_spring_03 {parent=dyn_spring_02, loc=<4 0 0>, rot=<0 0 0 1>}
 dyn_spring_04 {parent=dyn_spring_03, loc=<4 0 0>, rot=<0 0 0 1>}
 dyn_spring_05 {parent=dyn_spring_04, loc=<4 0 0>, rot=<0 0 0.0003 -1>}
 dyn_head {parent=dyn_spring_05, loc=<3.0775 0 0>, rot=<0 0 -0.0003 -1>}
 dyn_hat_r_MSTR {parent=dyn_head, loc=<8.0359 2.7652 -0.0196>, rot=<-0.6498 -0.2691 -0.265 -0.6596>}
 dyn_hat_r_01 {parent=dyn_hat_r_MSTR, loc=<0.8044 0 0>, rot=<-0.0119 -0.0494 -0.0129 -0.9986>}
 dyn_hat_r_02 {parent=dyn_hat_r_01, loc=<1.9156 0 0>, rot=<0.0035 -0.3495 0.0098 -0.9369>}
 dyn_hat_r_03 {parent=dyn_hat_r_02, loc=<2.3722 0 0>, rot=<0.0001 -0.2833 -0.0003 -0.959>}
 dyn_hat_r_04 {parent=dyn_hat_r_03, loc=<2.2037 0 0>, rot=<0.0001 -0.2297 -0.0001 -0.9733>}
 dyn_hat_l_MSTR {parent=dyn_head, loc=<8.0359 -2.7624 -0.0196>, rot=<0.6498 -0.2689 0.2649 -0.6597>}
 dyn_hat_l_01 {parent=dyn_hat_l_MSTR, loc=<0.804 0 0>, rot=<0.012 -0.0496 0.0129 -0.9986>}
 dyn_hat_l_02 {parent=dyn_hat_l_01, loc=<1.9156 0 0>, rot=<-0.0035 -0.3495 -0.0098 -0.9369>}
 dyn_hat_l_03 {parent=dyn_hat_l_02, loc=<2.3723 0 0>, rot=<-0.0001 -0.2833 0.0003 -0.959>}
 dyn_hat_l_04 {parent=dyn_hat_l_03, loc=<2.204 0 0>, rot=<-0.0001 -0.2297 0.0001 -0.9733>}
 dyn_l_grenade_MSTR {parent=dyn_spring_05, loc=<2.4302 -2.4247 -0.0016>, rot=<0.3187 -0.372 0.8608 -0.1379>}
 dyn_l_grenade_01 {parent=dyn_l_grenade_MSTR, loc=<1.9809 0 0>, rot=<0.3178 0.1325 -0.3701 -0.8628>}
 dyn_l_grenade_pin_MSTR {parent=dyn_l_grenade_01, loc=<1.2665 1.0135 0.9278>, rot=<0.575 0.2503 -0.1645 -0.7613>}
 dyn_l_grenade_pin {parent=dyn_l_grenade_pin_MSTR, loc=<0.0998 0 0.104>, rot=<-0.0243 0 0 -0.9997>}
 dyn_r_grenade_MSTR {parent=dyn_spring_05, loc=<2.4277 2.4273 -0.0016>, rot=<0.5137 0.4632 0.6063 0.3924>}
 dyn_r_grenade_01 {parent=dyn_r_grenade_MSTR, loc=<2.3017 0 0>, rot=<-0.1311 -0.2746 0.267 -0.9144>}
 dyn_r_grenade_pin_MSTR {parent=dyn_r_grenade_01, loc=<1.2643 -1.369 0.1395>, rot=<-0.4221 0.2454 0.1727 -0.8555>}
 dyn_r_grenade_pin {parent=dyn_r_grenade_pin_MSTR, loc=<0.0993 -0.031 0.133>, rot=<-0.0542 0 0 -0.9985>}
 dyn_box_r_MSTR {parent=pack, loc=<8.7214 -7.2387 5.2675>, rot=<0 -0.258 0 -0.9661>}
 dyn_box_r {parent=dyn_box_r_MSTR, loc=<0 0 0>, rot=<0 0 0 1>}
 dyn_box_l_MSTR {parent=pack, loc=<-8.7214 -7.2387 5.2675>, rot=<-0.258 0 0.9661 0>}
 dyn_box_l {parent=dyn_box_l_MSTR, loc=<0 0 0>, rot=<0 0 0 -1>}
 dyn_faceplate_MSTR {parent=pack, loc=<0 -14.8427 -10.4318>, rot=<-0.3174 -0.3174 0.6319 -0.6319>}
 dyn_faceplate {parent=dyn_faceplate_MSTR, loc=<0 0 0>, rot=<0 0 0 1>}
 dyn_band {parent=pack, loc=<8.2534 -7.2517 -9.5981>, rot=<0 -0.7071 0 -0.7071>}
 dyn_button_top {parent=pack, loc=<10.9551 -7.2387 -1.8522>, rot=<0 -0.0544 0 -0.9985>}
 dyn_button_bottom {parent=pack, loc=<10.6318 -7.2387 -5.279>, rot=<0 -0.0515 0 -0.9987>}
]
