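# Physics asset for the Cotton Candy back bling in its box variant, defining collision and joint limits for its dynamic band and faceplate bones.
# Capsule bodies sit on dyn_band and dyn_faceplate, with a tiny capsule on the pack bone as their shared anchor.
# Each dyn bone is constrained directly to pack with tight twist limits; the band swings more freely, while the faceplate moves mostly along one swing axis.
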PhysicsAsset /BRCosmetics/Accessories/FORT_Backpacks/Backpack_F_MED_Cotton_Candy/Meshes/F_MED_Cotton_Candy_Pack_Physics_Box
bodies {
 dyn_band [{shapeType=Capsule, length=3.0551, radius=1.7556, center=<2.5 0 0>, rotation=<rot 90 0 0>}]
 dyn_faceplate [{shapeType=Capsule, length=8.5349, radius=3.2214, center=<2.58 0 0>, rotation=<rot 0 0 90>}]
 pack [{shapeType=Capsule, length=0.505, radius=0.505}]
}
constraints [
 {bone1Name=dyn_faceplate, bone2Name=pack, swing1LimitDegrees=3, swing2LimitDegrees=15, twistLimitDegrees=1}
 {bone1Name=dyn_band, bone2Name=pack, swing1LimitDegrees=25, swing2LimitDegrees=25, twistLimitDegrees=3}
]
