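# Physics asset for the Cotton Candy backpack cosmetic, giving its left and right box parts bodies for physical motion.
# A large box covers the pack, with a thin box for each side part and a tiny capsule on each side's MSTR bone.
# Each side chains box to MSTR to pack: the box joint has tight angular limits with more room on one swing axis, while the MSTR-to-pack joint rotates freely.
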PhysicsAsset /BRCosmetics/Accessories/FORT_Backpacks/Backpack_F_MED_Cotton_Candy/Meshes/F_MED_Cotton_Candy_Pack_new_box_Physics
bodies {
 dyn_box_l [{shapeType=Box, extentX=9.3136, extentY=1.4282, extentZ=16.1477, center=<4.2851 -0.0016 -0.0059>, rotation=<rot 0 -179.9397 -89.7134>}]
 dyn_box_l_MSTR [{shapeType=Capsule, length=0.505, radius=0.505, rotation=<rot 0 -90 0>}]
 dyn_box_r [{shapeType=Box, extentX=9.3136, extentY=1.4282, extentZ=16.1477, center=<4.2851 0.0016 -0.0059>, rotation=<rot 0 -0.0603 -89.7134>}]
 dyn_box_r_MSTR [{shapeType=Capsule, length=0.505, radius=0.505, rotation=<rot 0 -90 0>}]
 pack [{shapeType=Box, extentX=33.4044, extentY=7.8058, extentZ=42.2086, center=<0 7.1954 8.0585>, rotation=<rot 0 0 6.2504>}]
}
constraints [
 {bone1Name=dyn_box_l, bone2Name=dyn_box_l_MSTR, swing1LimitDegrees=1, swing2LimitDegrees=15, twistLimitDegrees=1}
 {bone1Name=dyn_box_r, bone2Name=dyn_box_r_MSTR, swing1LimitDegrees=1, swing2LimitDegrees=15, twistLimitDegrees=1}
 {
  bone1Name=dyn_box_r_MSTR
  bone2Name=pack
  swing1LimitDegrees=45
  swing1Motion=Free
  swing2LimitDegrees=45
  swing2Motion=Free
  twistLimitDegrees=45
  twistMotion=Free
 }
 {
  bone1Name=dyn_box_l_MSTR
  bone2Name=pack
  swing1LimitDegrees=45
  swing1Motion=Free
  swing2LimitDegrees=45
  swing2Motion=Free
  twistLimitDegrees=45
  twistMotion=Free
 }
]
