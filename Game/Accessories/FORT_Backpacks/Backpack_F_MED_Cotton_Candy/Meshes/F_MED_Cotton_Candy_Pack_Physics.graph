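# Physics asset for the Cotton Candy backpack, giving its spring-mounted head, two-sided hat, grenades and side boxes collision bodies and joints for secondary motion.
# A narrowly limited five-segment spring chain carries the head and both grenades, each with a pin.
# Two four-segment hat chains branch from the head, loosening toward the tips.
# Small MSTR bodies join the spring and boxes to the pack box with free swing and twist.
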PhysicsAsset /Game/Accessories/FORT_Backpacks/Backpack_F_MED_Cotton_Candy/Meshes/F_MED_Cotton_Candy_Pack_Physics
bodies {
 dyn_box_l [{shapeType=Box, extentX=9.3136, extentY=1.4282, extentZ=16.1477, center=<4.2851 -0.0016 -0.0059>, rotation=<rot 0 -179.9397 -89.7134>}]
 dyn_box_l_MSTR [{shapeType=Capsule, length=0.505, radius=0.505, rotation=<rot 0 -90 0>}]
 dyn_box_r [{shapeType=Box, extentX=9.3136, extentY=1.4282, extentZ=16.1477, center=<4.2851 0.0016 -0.0059>, rotation=<rot 0 -0.0603 -89.7134>}]
 dyn_box_r_MSTR [{shapeType=Capsule, length=0.505, radius=0.505, rotation=<rot 0 -90 0>}]
 dyn_hat_l_01 [{shapeType=Capsule, length=1.4718, radius=1.7177, center=<1 0 0>, rotation=<rot 90 0 0>}]
 dyn_hat_l_02 [{shapeType=Capsule, length=1.2982, radius=1.2351, center=<1.5 0 0>, rotation=<rot 90 0 0>}]
 dyn_hat_l_03 [{shapeType=Capsule, length=1.6516, radius=1.0602, center=<1 0 0>, rotation=<rot 90 0 0>}]
 dyn_hat_l_04 [{shapeType=Sphere, radius=1.8219, center=<1.7697 -0.0032 -0.0218>}]
 dyn_hat_r_01 [{shapeType=Capsule, length=1.4718, radius=1.7177, center=<1 0 0>, rotation=<rot 90 0 0>}]
 dyn_hat_r_02 [{shapeType=Capsule, length=1.2982, radius=1.2351, center=<1.5 0 0>, rotation=<rot 90 0 0>}]
 dyn_hat_r_03 [{shapeType=Capsule, length=1.6516, radius=1.0602, center=<1 0 0>, rotation=<rot 90 0 0>}]
 dyn_hat_r_04 [{shapeType=Sphere, radius=1.8219, center=<1.7701 0.0032 -0.0215>}]
 dyn_head [{shapeType=Sphere, radius=5, center=<5.1821 0 0>}]
 dyn_l_grenade_01 [{shapeType=Capsule, length=4.1811, radius=3.6711, center=<5.8141 0 0>, rotation=<rot 90 0 0>}]
 dyn_l_grenade_pin [{shapeType=Box, extentX=4, extentY=4, extentZ=1, center=<2 0 2>}]
 dyn_r_grenade_01 [{shapeType=Capsule, length=4.8511, radius=3.6711, center=<5.5 0 0>, rotation=<rot 90 0 0>}]
 dyn_r_grenade_pin [{shapeType=Box, extentX=4, extentY=4, extentZ=1, center=<2 0 2>}]
 dyn_spring_01 [{shapeType=Capsule, length=1.0312, radius=2.8824, center=<2 0 0>, rotation=<rot 90 0 0>}]
 dyn_spring_02 [{shapeType=Capsule, length=1.0822, radius=2.6576, center=<2 0 0>, rotation=<rot 90 0 0>}]
 dyn_spring_03 [{shapeType=Capsule, length=1.1079, radius=2.5367, center=<2 0 0>, rotation=<rot 90 0 0>}]
 dyn_spring_04 [{shapeType=Capsule, length=2.5728, radius=2.4819, center=<2 0 0>, rotation=<rot 90 0 0>}]
 dyn_spring_05 [{shapeType=Capsule, length=0.9512, radius=2.5236, center=<2 0 0>, rotation=<rot 90 0 0>}]
 dyn_spring_MSTR [{shapeType=Capsule, length=0.505, radius=0.505}]
 pack [{shapeType=Box, extentX=33.4044, extentY=7.8058, extentZ=42.2086, center=<0 7.1954 8.0585>, rotation=<rot 0 0 6.2504>}]
}
constraints [
 {bone1Name=dyn_spring_02, bone2Name=dyn_spring_01, swing1LimitDegrees=1, swing2LimitDegrees=1, twistLimitDegrees=1}
 {bone1Name=dyn_spring_03, bone2Name=dyn_spring_02, swing1LimitDegrees=1, swing2LimitDegrees=5, twistLimitDegrees=1}
 {bone1Name=dyn_spring_04, bone2Name=dyn_spring_03, swing1LimitDegrees=1, swing2LimitDegrees=5, twistLimitDegrees=1}
 {bone1Name=dyn_spring_05, bone2Name=dyn_spring_04, swing1LimitDegrees=1, swing2LimitDegrees=5, twistLimitDegrees=1}
 {bone1Name=dyn_head, bone2Name=dyn_spring_05, swing1LimitDegrees=5, swing2LimitDegrees=5, twistLimitDegrees=3}
 {bone1Name=dyn_hat_r_01, bone2Name=dyn_head, swing1LimitDegrees=20, swing2LimitDegrees=20, twistLimitDegrees=3}
 {bone1Name=dyn_hat_r_02, bone2Name=dyn_hat_r_01, swing1LimitDegrees=20, swing2LimitDegrees=20, twistLimitDegrees=3}
 {bone1Name=dyn_hat_r_03, bone2Name=dyn_hat_r_02, swing1LimitDegrees=25, swing2LimitDegrees=25, twistLimitDegrees=3}
 {bone1Name=dyn_hat_r_04, bone2Name=dyn_hat_r_03, swing1LimitDegrees=30, swing2LimitDegrees=30, twistLimitDegrees=3}
 {bone1Name=dyn_hat_l_01, bone2Name=dyn_head, swing1LimitDegrees=20, swing2LimitDegrees=20, twistLimitDegrees=3}
 {bone1Name=dyn_hat_l_02, bone2Name=dyn_hat_l_01, swing1LimitDegrees=20, swing2LimitDegrees=20, twistLimitDegrees=3}
 {bone1Name=dyn_hat_l_03, bone2Name=dyn_hat_l_02, swing1LimitDegrees=25, swing2LimitDegrees=25, twistLimitDegrees=3}
 {bone1Name=dyn_hat_l_04, bone2Name=dyn_hat_l_03, swing1LimitDegrees=30, swing2LimitDegrees=30, twistLimitDegrees=3}
 {bone1Name=dyn_l_grenade_01, bone2Name=dyn_spring_05, swing1LimitDegrees=25, swing2LimitDegrees=25, twistLimitDegrees=3}
 {bone1Name=dyn_r_grenade_01, bone2Name=dyn_spring_05, swing1LimitDegrees=25, swing2LimitDegrees=25, twistLimitDegrees=3}
 {bone1Name=dyn_spring_01, bone2Name=dyn_spring_MSTR, swing1LimitDegrees=1, swing2LimitDegrees=1, twistLimitDegrees=1}
 {bone1Name=dyn_box_l, bone2Name=dyn_box_l_MSTR, swing1LimitDegrees=1, swing2LimitDegrees=15, twistLimitDegrees=1}
 {bone1Name=dyn_box_r, bone2Name=dyn_box_r_MSTR, swing1LimitDegrees=1, swing2LimitDegrees=15, twistLimitDegrees=1}
 {
  bone1Name=dyn_spring_MSTR
  bone2Name=pack
  swing1LimitDegrees=45
  swing1Motion=Free
  swing2LimitDegrees=45
  swing2Motion=Free
  twistLimitDegrees=45
  twistMotion=Free
 }
 {
  bone1Name=dyn_box_r_MSTR
  bone2Name=pack
  swing1LimitDegrees=45
  swing1Motion=Free
  swing2LimitDegrees=45
  swing2Motion=Free
  twistLimitDegrees=45
  twistMotion=Free
 }
 {
  bone1Name=dyn_box_l_MSTR
  bone2Name=pack
  swing1LimitDegrees=45
  swing1Motion=Free
  swing2LimitDegrees=45
  swing2Motion=Free
  twistLimitDegrees=45
  twistMotion=Free
 }
 {bone1Name=dyn_l_grenade_pin, bone2Name=dyn_l_grenade_01, swing1LimitDegrees=25, swing2LimitDegrees=25, twistLimitDegrees=3}
 {bone1Name=dyn_r_grenade_pin, bone2Name=dyn_r_grenade_01, swing1LimitDegrees=25, swing2LimitDegrees=25, twistLimitDegrees=3}
]
