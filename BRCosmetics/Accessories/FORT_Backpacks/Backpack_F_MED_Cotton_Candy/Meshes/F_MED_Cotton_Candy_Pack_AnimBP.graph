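# Animation Blueprint for the Cotton Candy backpack, which pops out a sprung head for a surprise when the player's kill count rises.
# The EventGraph watches the kill stat, triggers after a random delay, drives scale, wobble and upward force, and fires open/close FX dispatchers around a tick-counted timer.
# AnimGraph blends between RigidBody plus AnimDynamics chains when out and a scaled spring with Control Rig when closed, leaving several nodes disconnected.
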AnimBlueprint /BRCosmetics/Accessories/FORT_Backpacks/Backpack_F_MED_Cotton_Candy/Meshes/F_MED_Cotton_Candy_Pack_AnimBP
target_skeleton=&/F_MED_Cotton_Candy_Pack_Skeleton
graphs [
 AnimGraph {nodes [
   Anim Dynamics - Bone: dyn_spring_03!(ComponentPose, bAlphaBoolEnabled:bool=True, ExternalForce="0,0,0") -> (Pose) {
    alphaInputType=Bool
    bGravityOverrideInSimSpace=true
    bLinearSpring=true
    bOverrideLinearDamping=true
    gravityScale=0.4
    linearDampingOverride=0.9999
    linearSpringConstant=300
    simulationSpace=BoneRelative
    alphaBoolBlend {blendInTime=0.2, blendOption=Linear, blendOutTime=0.2}
    boundBone {boneName=dyn_spring_03}
    componentAppliedLinearAccClamp=<2000 2000 2000>
    componentLinearAccScale=<0.4 0.4 0.4>
    componentLinearVelScale=<0.4 0.4 0.4>
    physicsBodyDefinitions [{
      sphereCollisionRadius=0
      boundBone {boneName=dyn_spring_03}
      boxExtents=<3 3 3>
      constraintSetup {angularLimitsMax=<0 0 0>, angularLimitsMin=<0 0 0>, angularTarget=<0 0 0>, linearAxesMax=<5.7 0 0>}
      localJointOffset=<5 0 0>
     }]
    relativeSpaceBone {boneName=pack}
   }
   n1 = Get Surprise_Out() -> (Surprise_Out:bool)
   n2 = AND Boolean(A=@n1, B=OR Boolean(A=Get bIsJumping(), B=Get bIsFalling()))
   n3 = "Transform (Modify) Bone - Bone: dyn_l_grenade_MSTR"!(ComponentPose, bAlphaBoolEnabled:bool=@n2, Translation="0,0,0", Rotation="20,0,0", Scale="1,1,1") -> (Pose) {
    alphaInputType=Bool
    rotationMode=BMM_Additive
    scaleSpace=BCS_BoneSpace
    alphaBoolBlend {blendInTime=1, blendOption=Linear, blendOutTime=0.5}
    boneToModify {boneName=dyn_l_grenade_MSTR}
    rotation {pitch=20}
   }
   n4 = "Transform (Modify) Bone - Bone: dyn_r_grenade_MSTR"!(ComponentPose=@n3, bAlphaBoolEnabled:bool=@n2, Translation="0,0,0", Rotation="30,0,0", Scale="1,1,1") -> (Pose) {
    alphaInputType=Bool
    rotationMode=BMM_Additive
    scaleSpace=BCS_BoneSpace
    alphaBoolBlend {blendInTime=1, blendOption=Linear, blendOutTime=0.5}
    boneToModify {boneName=dyn_r_grenade_MSTR}
    rotation {pitch=30}
   }
   n5 = Anim Dynamics - Bone: dyn_spring_01!(ComponentPose=@n4, bAlphaBoolEnabled:bool=@n1, ExternalForce="0,0,0") -> (Pose) {
    alphaInputType=Bool
    bGravityOverrideInSimSpace=true
    bLinearSpring=true
    bOverrideLinearDamping=true
    gravityScale=0.4
    linearDampingOverride=0.9999
    linearSpringConstant=300
    simulationSpace=BoneRelative
    alphaBoolBlend {blendInTime=0.2, blendOption=Linear, blendOutTime=0.2}
    boundBone {boneName=dyn_spring_01}
    componentAppliedLinearAccClamp=<2000 2000 2000>
    componentLinearAccScale=<0.4 0.4 0.4>
    componentLinearVelScale=<0.4 0.4 0.4>
    physicsBodyDefinitions [{
      sphereCollisionRadius=0
      boundBone {boneName=dyn_spring_01}
      boxExtents=<1 1 1>
      constraintSetup {angularLimitsMax=<0 0 0>, angularLimitsMin=<0 0 0>, angularTarget=<0 0 0>, linearAxesMax=<5.7 0 0>}
      localJointOffset=<0 0 0>
     }]
    relativeSpaceBone {boneName=pack}
   }
   n6 = Anim Dynamics - Bone: dyn_spring_02!(ComponentPose=@n5, bAlphaBoolEnabled:bool=@n1, ExternalForce="0,0,0") -> (Pose) {
    alphaInputType=Bool
    bGravityOverrideInSimSpace=true
    bLinearSpring=true
    bOverrideLinearDamping=true
    gravityScale=0.4
    linearDampingOverride=0.9999
    linearSpringConstant=300
    simulationSpace=BoneRelative
    alphaBoolBlend {blendInTime=0.2, blendOption=Linear, blendOutTime=0.2}
    boundBone {boneName=dyn_spring_02}
    componentAppliedLinearAccClamp=<2000 2000 2000>
    componentLinearAccScale=<0.4 0.4 0.4>
    componentLinearVelScale=<0.4 0.4 0.4>
    physicsBodyDefinitions [{
      sphereCollisionRadius=0
      boundBone {boneName=dyn_spring_02}
      boxExtents=<1 1 1>
      constraintSetup {angularLimitsMax=<0 0 0>, angularLimitsMin=<0 0 0>, angularTarget=<0 0 0>, linearAxesMax=<4.3 0 0>}
      localJointOffset=<0 0 0>
     }]
    relativeSpaceBone {boneName=pack}
   }
   n7 = Anim Dynamics - Bone: dyn_spring_03(ComponentPose=@n6, bAlphaBoolEnabled=@n1, ExternalForce="0,0,0") {
    alphaInputType=Bool
    bGravityOverrideInSimSpace=true
    bLinearSpring=true
    bOverrideLinearDamping=true
    gravityScale=0.4
    linearDampingOverride=0.9999
    linearSpringConstant=300
    simulationSpace=BoneRelative
    alphaBoolBlend {blendInTime=0.3, blendOption=Linear, blendOutTime=0.2}
    boundBone {boneName=dyn_spring_03}
    componentAppliedLinearAccClamp=<2000 2000 2000>
    componentLinearAccScale=<0.4 0.4 0.4>
    componentLinearVelScale=<0.4 0.4 0.4>
    physicsBodyDefinitions [{
      sphereCollisionRadius=0
      boundBone {boneName=dyn_spring_03}
      boxExtents=<1 1 1>
      constraintSetup {angularLimitsMax=<0 0 0>, angularLimitsMin=<0 0 0>, angularTarget=<0 0 0>, linearAxesMax=<2.9 0 0>, linearAxesMin=<-0.5 0 0>}
      localJointOffset=<0 0 0>
     }]
    relativeSpaceBone {boneName=pack}
   }
   Anim Dynamics - Bone: dyn_spring_04!(ComponentPose=@n7, bAlphaBoolEnabled:bool=@n1, ExternalForce="0,0,0") -> (Pose) {
    alphaInputType=Bool
    bGravityOverrideInSimSpace=true
    bLinearSpring=true
    bOverrideLinearDamping=true
    gravityScale=0.4
    linearDampingOverride=0.9999
    linearSpringConstant=300
    simulationSpace=BoneRelative
    alphaBoolBlend {blendInTime=0.3, blendOption=Linear, blendOutTime=0.2}
    boundBone {boneName=dyn_spring_04}
    componentAppliedLinearAccClamp=<2000 2000 2000>
    componentLinearAccScale=<0.4 0.4 0.4>
    componentLinearVelScale=<0.4 0.4 0.4>
    physicsBodyDefinitions [{
      sphereCollisionRadius=0
      boundBone {boneName=dyn_spring_04}
      boxExtents=<1 1 1>
      constraintSetup {angularLimitsMax=<0 0 0>, angularLimitsMin=<0 0 0>, angularTarget=<0 0 0>, linearAxesMax=<1.44 0 0>}
      localJointOffset=<0 0 0>
     }]
    relativeSpaceBone {boneName=pack}
   }
   Anim Dynamics - Bone: dyn_l_grenade_01(Alpha=1, ExternalForce="0,0,0") {
    angularDampingOverride=0.998
    bGravityOverrideInSimSpace=true
    bOverrideAngularDamping=true
    bUseGravityOverride=true
    simulationSpace=RootRelative
    alphaBoolBlend {blendOption=Linear}
    boundBone {boneName=dyn_l_grenade_01}
    componentAppliedLinearAccClamp=<3000 3000 3000>
    componentLinearAccScale=<0.3 0.3 0.3>
    componentLinearVelScale=<0.3 0.3 0.3>
    gravityOverride {z=-980}
    physicsBodyDefinitions [{
      sphereCollisionRadius=0
      boundBone {boneName=dyn_l_grenade_01}
      boxExtents=<5 10 10>
      constraintSetup {angularLimitsMax=<1 0 15>, angularLimitsMin=<-1 -45 -15>, angularTarget=<0 0 0>}
      localJointOffset=<2 0 0>
     }]
    relativeSpaceBone {boneName=dyn_spring_04}
   }
   n8 = "Transform (Modify) Bone - Bone: dyn_box_r_MSTR"(Alpha=1, Translation="-0.157,0.000,0.000", Rotation="0,0,0", Scale="1.000,0.965,1.000") {
    rotationMode=BMM_Additive
    scaleMode=BMM_Additive
    translationMode=BMM_Additive
    alphaBoolBlend {blendOption=Linear}
    boneToModify {boneName=dyn_box_r_MSTR}
    scale {y=0.9647}
    translation=<-0.1573 0>
   }
   n9 = "Transform (Modify) Bone - Bone: dyn_box_l_MSTR"(ComponentPose=@n8, Alpha=0, Translation="0.157,0.000,0", Rotation="0,0,0", Scale="1.000,0.965,1.000") {
    rotationMode=BMM_Additive
    scaleMode=BMM_Additive
    translationMode=BMM_Additive
    alphaBoolBlend {blendOption=Linear}
    boneToModify {boneName=dyn_box_l_MSTR}
    scale {y=0.9647}
    translation {x=0.157}
   }
   Component To Local(ComponentPose=@n9)
   Local To Component()
   Local To Component()
   n10 = Get initial transfrom() -> (initial transfrom:double)
   n11 = "Transform (Modify) Bone - Bone: dyn_spring_01"(Alpha=@n10, Translation="0,0,2", Rotation="0,0,0", Scale="1,1,1") {
    scaleSpace=BCS_BoneSpace
    translationMode=BMM_Additive
    alphaBoolBlend {blendInTime=0.2, blendOption=Linear, blendOutTime=0.2}
    boneToModify {boneName=dyn_spring_01}
    translation {z=2}
   }
   "Transform (Modify) Bone - Bone: dyn_spring_03"(ComponentPose=@n11, Alpha=@n10, Translation="0,0,2", Rotation="0,0,0", Scale="1,1,1") {
    scaleSpace=BCS_BoneSpace
    translationMode=BMM_Additive
    alphaBoolBlend {blendInTime=0.2, blendOption=Linear, blendOutTime=0.2}
    boneToModify {boneName=dyn_spring_03}
    translation {z=2}
   }
   n12 = Get physics_out() -> (physics_out:bool)
   RigidBody!(ComponentPose, bAlphaBoolEnabled:bool=@n12, OverrideWorldGravity:Vector=Get GravityOverride(), ExternalForce) -> (Pose) {
    alphaInputType=Bool
    bClampLinearTranslationLimitToRefPose=true
    bOverrideWorldGravity=true
    overridePhysicsAsset=/Game/Accessories/FORT_Backpacks/Backpack_F_MED_Cotton_Candy/Meshes/F_MED_Cotton_Candy_Pack_Physics
    simulationSpace=BaseBoneSpace
    alphaBoolBlend {blendInTime=0.15, blendOption=Linear, blendOutTime=0.1}
    baseBoneRef {boneName=pack}
    componentAppliedLinearAccClamp=<3000 3000 3000>
    componentLinearAccScale=<0.25 0.25 0.2>
    componentLinearVelScale=<0.2 0.2 0.2>
    overrideWorldGravity {x=200, z=-200}
   }
   n13 = RigidBody!(ComponentPose=Local To Component(LocalPose=Local Space Ref Pose()), Alpha:float, ExternalForce) -> (Pose) {
    bClampLinearTranslationLimitToRefPose=true
    bOverrideWorldGravity=true
    overridePhysicsAsset=&/F_MED_Cotton_Candy_Pack_Physics_Box
    simulationSpace=BaseBoneSpace
    alphaBoolBlend {blendOption=Linear}
    baseBoneRef {boneName=pack}
    componentLinearAccScale=<0.3 0.3 0.5>
    componentLinearVelScale=<0.2 0.2 0.3>
    overrideWorldGravity {z=-980}
   }
   n14 = RigidBody!(ComponentPose=@n13, bAlphaBoolEnabled:bool, OverrideWorldGravity:Vector=Get GravityOverride(), ExternalForce) -> (Pose) {
    alphaInputType=Bool
    bClampLinearTranslationLimitToRefPose=true
    bOverrideWorldGravity=true
    overridePhysicsAsset=&/F_MED_Cotton_Candy_Pack_new_box_Physics
    simulationSpace=BaseBoneSpace
    alphaBoolBlend {blendInTime=0.15, blendOption=Linear, blendOutTime=0.1}
    baseBoneRef {boneName=pack}
    componentAppliedLinearAccClamp=<3000 3000 3000>
    componentLinearAccScale=<0.25 0.25 0.2>
    componentLinearVelScale=<0.2 0.2 0.2>
    overrideWorldGravity {x=200, z=-200}
   }
   n15 = Anim Dynamics - Bone: dyn_spring_03(ComponentPose=@n14, bAlphaBoolEnabled=True, ExternalForce="0,0,0") {
    alphaInputType=Bool
    bGravityOverrideInSimSpace=true
    bLinearSpring=true
    bOverrideLinearDamping=true
    bUseGravityOverride=true
    gravityScale=0.4
    linearDampingOverride=0.9999
    linearSpringConstant=150
    simulationSpace=BoneRelative
    alphaBoolBlend {blendInTime=0.3, blendOption=Linear, blendOutTime=0.2}
    boundBone {boneName=dyn_spring_03}
    componentAppliedLinearAccClamp=<1000 1000 1000>
    componentLinearAccScale=<0.25 0.25 0.2>
    componentLinearVelScale=<0.15 0.15 0.2>
    gravityOverride {z=-200}
    physicsBodyDefinitions [{
      sphereCollisionRadius=0
      boundBone {boneName=dyn_spring_03}
      boxExtents=<5 5 5>
      constraintSetup {angularLimitsMax=<0 0 0>, angularLimitsMin=<0 0 0>, angularTarget=<0 0 0>, linearAxesMax=<2 0 0>, linearAxesMin=<-1 0 0>}
      localJointOffset=<-1.5 0 0>
     }]
    relativeSpaceBone {boneName=pack}
   }
   n16 = Get Surprise Force() -> (Surprise Force:Vector)
   n17 = Anim Dynamics - Bone: dyn_spring_04(ComponentPose=@n15, bAlphaBoolEnabled=True, ExternalForce=@n16) {
    alphaInputType=Bool
    bGravityOverrideInSimSpace=true
    bLinearSpring=true
    bOverrideLinearDamping=true
    bUseGravityOverride=true
    gravityScale=0.4
    linearDampingOverride=0.9999
    linearSpringConstant=150
    simulationSpace=BoneRelative
    alphaBoolBlend {blendInTime=0.3, blendOption=Linear, blendOutTime=0.2}
    boundBone {boneName=dyn_spring_04}
    componentAppliedLinearAccClamp=<1500 1500 1000>
    componentLinearAccScale=<0.25 0.25 0.3>
    componentLinearVelScale=<0.15 0.15 0.2>
    gravityOverride {z=-200}
    physicsBodyDefinitions [{
      sphereCollisionRadius=0
      boundBone {boneName=dyn_spring_04}
      boxExtents=<5 5 5>
      constraintSetup {angularLimitsMax=<0 0 0>, angularLimitsMin=<0 0 0>, angularTarget=<0 0 0>, linearAxesMax=<2 0 0>, linearAxesMin=<-1 0 0>}
      localJointOffset=<-3 0 0>
     }]
    relativeSpaceBone {boneName=pack}
   }
   n18 = Anim Dynamics - Bone: dyn_spring_05!(ComponentPose=@n17, bAlphaBoolEnabled:bool=True, ExternalForce=@n16) -> (Pose) {
    alphaInputType=Bool
    bGravityOverrideInSimSpace=true
    bLinearSpring=true
    bOverrideLinearDamping=true
    bUseGravityOverride=true
    gravityScale=0.4
    linearDampingOverride=0.9999
    linearSpringConstant=150
    simulationSpace=BoneRelative
    alphaBoolBlend {blendInTime=0.3, blendOption=Linear, blendOutTime=0.2}
    boundBone {boneName=dyn_spring_05}
    componentAppliedLinearAccClamp=<2500 2500 2500>
    componentLinearAccScale=<0.25 0.25 0.4>
    componentLinearVelScale=<0.15 0.15 0.2>
    gravityOverride {z=-200}
    physicsBodyDefinitions [{
      sphereCollisionRadius=0
      boundBone {boneName=dyn_spring_05}
      boxExtents=<5 5 5>
      constraintSetup {angularLimitsMax=<0 0 0>, angularLimitsMin=<0 0 0>, angularTarget=<0 0 0>, linearAxesMax=<3 0 0>, linearAxesMin=<-1 0 0>}
      localJointOffset=<-2.5 0 0>
     }]
    relativeSpaceBone {boneName=pack}
   }
   n19 = Anim Dynamics - Bone: dyn_head!(ComponentPose=@n18, bAlphaBoolEnabled:bool=True, ExternalForce=@n16) -> (Pose) {
    alphaInputType=Bool
    bGravityOverrideInSimSpace=true
    bLinearSpring=true
    bOverrideLinearDamping=true
    bUseGravityOverride=true
    gravityScale=0.4
    linearDampingOverride=0.9999
    linearSpringConstant=150
    simulationSpace=BoneRelative
    alphaBoolBlend {blendInTime=0.3, blendOption=Linear, blendOutTime=0.2}
    boundBone {boneName=dyn_head}
    componentAppliedLinearAccClamp=<2000 2000 2000>
    componentLinearAccScale=<0.25 0.25 0.3>
    componentLinearVelScale=<0.15 0.15 0.2>
    gravityOverride {z=-200}
    physicsBodyDefinitions [{
      sphereCollisionRadius=0
      boundBone {boneName=dyn_head}
      boxExtents=<5 5 5>
      constraintSetup {angularLimitsMax=<0 0 0>, angularLimitsMin=<0 0 0>, angularTarget=<0 0 0>, linearAxesMax=<1 0 0>, linearAxesMin=<-1 0 0>}
      localJointOffset=<-1.5 0 0>
     }]
    relativeSpaceBone {boneName=pack}
   }
   n20 = Anim Dynamics - Bone: dyn_spring_03!(ComponentPose=@n19, Alpha:float=1, ExternalForce="0,0,0") -> (Pose) {
    angularDampingOverride=0.98
    angularSpringConstant=1500
    bAngularSpring=true
    bGravityOverrideInSimSpace=true
    bOverrideAngularDamping=true
    bUseGravityOverride=true
    simulationSpace=RootRelative
    alphaBoolBlend {blendOption=Linear}
    boundBone {boneName=dyn_spring_03}
    componentLinearAccScale=<0.7 0.7 0.7>
    componentLinearVelScale=<0.7 0.7 0.7>
    gravityOverride {z=-100}
    physicsBodyDefinitions [{
      sphereCollisionRadius=0
      boundBone {boneName=dyn_spring_03}
      boxExtents=<20 25 25>
      constraintSetup {angularLimitsMax=<1 5 3>, angularLimitsMin=<-1 -1 -3>, angularTarget=<1 0 0>}
      localJointOffset=<1.5 0 0>
     }]
   }
   n21 = Anim Dynamics - Bone: dyn_spring_04!(ComponentPose=@n20, Alpha:float=1, ExternalForce="0,0,0") -> (Pose) {
    angularDampingOverride=0.98
    angularSpringConstant=1500
    bAngularSpring=true
    bGravityOverrideInSimSpace=true
    bOverrideAngularDamping=true
    bUseGravityOverride=true
    simulationSpace=RootRelative
    alphaBoolBlend {blendOption=Linear}
    boundBone {boneName=dyn_spring_04}
    componentLinearAccScale=<0.7 0.7 0.7>
    componentLinearVelScale=<0.7 0.7 0.7>
    gravityOverride {z=-100}
    physicsBodyDefinitions [{
      sphereCollisionRadius=0
      boundBone {boneName=dyn_spring_04}
      boxExtents=<20 25 25>
      constraintSetup {angularLimitsMax=<1 5 3>, angularLimitsMin=<-1 -1 -3>, angularTarget=<1 0 0>}
      localJointOffset=<1.5 0 0>
     }]
   }
   n22 = Anim Dynamics - Bone: dyn_spring_05!(ComponentPose=@n21, Alpha:float=1, ExternalForce="0,0,0") -> (Pose) {
    angularDampingOverride=0.98
    angularSpringConstant=1500
    bAngularSpring=true
    bGravityOverrideInSimSpace=true
    bOverrideAngularDamping=true
    bUseGravityOverride=true
    simulationSpace=RootRelative
    alphaBoolBlend {blendOption=Linear}
    boundBone {boneName=dyn_spring_05}
    componentLinearAccScale=<0.7 0.7 0.7>
    componentLinearVelScale=<0.7 0.7 0.7>
    gravityOverride {z=-100}
    physicsBodyDefinitions [{
      sphereCollisionRadius=0
      boundBone {boneName=dyn_spring_05}
      boxExtents=<20 25 25>
      constraintSetup {angularLimitsMax=<1 10 3>, angularLimitsMin=<-1 -1 -3>, angularTarget=<1 0 0>}
      localJointOffset=<1.5 0 0>
     }]
   }
   n23 = Anim Dynamics - Bone: dyn_head!(ComponentPose=@n22, Alpha:float=1, ExternalForce="0,0,0") -> (Pose) {
    angularDampingOverride=0.98
    angularSpringConstant=1500
    bAngularSpring=true
    bGravityOverrideInSimSpace=true
    bOverrideAngularDamping=true
    bUseGravityOverride=true
    simulationSpace=RootRelative
    alphaBoolBlend {blendOption=Linear}
    boundBone {boneName=dyn_head}
    componentLinearAccScale=<0.7 0.7 0.7>
    componentLinearVelScale=<0.7 0.7 0.7>
    gravityOverride {z=-100}
    physicsBodyDefinitions [{
      sphereCollisionRadius=0
      boundBone {boneName=dyn_head}
      boxExtents=<10 15 15>
      constraintSetup {angularLimitsMax=<1 5 5>, angularLimitsMin=<-1 -5 -5>, angularTarget=<1 0 0>}
      localJointOffset=<1.5 0 0>
     }]
   }
   n24 = Anim Dynamics - Bone: dyn_hat_l_01(ComponentPose=@n23, Alpha=1, ExternalForce="0,0,0") {
    angularDampingOverride=0.98
    angularSpringConstant=300
    bAngularSpring=true
    bGravityOverrideInSimSpace=true
    bOverrideAngularDamping=true
    simulationSpace=RootRelative
    alphaBoolBlend {blendOption=Linear}
    boundBone {boneName=dyn_hat_l_01}
    componentLinearAccScale=<1.25 1.25 1.25>
    componentLinearVelScale=<1.25 1.25 1.25>
    physicsBodyDefinitions [{
      sphereCollisionRadius=0
      boundBone {boneName=dyn_hat_l_01}
      boxExtents=<5 20 20>
      constraintSetup {angularLimitsMax=<2 1 5>, angularLimitsMin=<-2 -25 -5>, angularTarget=<0 0 0>}
      localJointOffset=<1.5 0 0>
     }]
    relativeSpaceBone {boneName=root}
   }
   n25 = Anim Dynamics - Bone: dyn_hat_l_02(ComponentPose=@n24, Alpha=1, ExternalForce="0,0,0") {
    angularDampingOverride=0.98
    angularSpringConstant=300
    bAngularSpring=true
    bGravityOverrideInSimSpace=true
    bOverrideAngularDamping=true
    simulationSpace=RootRelative
    alphaBoolBlend {blendOption=Linear}
    boundBone {boneName=dyn_hat_l_02}
    componentLinearAccScale=<1.25 1.25 1.25>
    componentLinearVelScale=<1.25 1.25 1.25>
    physicsBodyDefinitions [{
      sphereCollisionRadius=0
      boundBone {boneName=dyn_hat_l_02}
      boxExtents=<5 15 15>
      constraintSetup {angularLimitsMax=<2 1 10>, angularLimitsMin=<-2 -25 -10>, angularTarget=<0 0 0>}
      localJointOffset=<1.5 0 0>
     }]
    relativeSpaceBone {boneName=root}
   }
   n26 = Anim Dynamics - Bone: dyn_hat_l_03(ComponentPose=@n25, Alpha=1, ExternalForce="0,0,0") {
    angularDampingOverride=0.98
    angularSpringConstant=300
    bAngularSpring=true
    bGravityOverrideInSimSpace=true
    bOverrideAngularDamping=true
    simulationSpace=RootRelative
    alphaBoolBlend {blendOption=Linear}
    boundBone {boneName=dyn_hat_l_03}
    componentLinearAccScale=<1.25 1.25 1.25>
    componentLinearVelScale=<1.25 1.25 1.25>
    physicsBodyDefinitions [{
      sphereCollisionRadius=0
      boundBone {boneName=dyn_hat_l_03}
      boxExtents=<5 15 15>
      constraintSetup {angularLimitsMax=<2 25 25>, angularLimitsMin=<-2 -25 -10>, angularTarget=<0 0 0>}
      localJointOffset=<1.5 0 0>
     }]
    relativeSpaceBone {boneName=root}
   }
   n27 = Anim Dynamics - Bone: dyn_hat_l_04(ComponentPose=@n26, Alpha=1, ExternalForce="0,0,0") {
    angularDampingOverride=0.98
    angularSpringConstant=300
    bAngularSpring=true
    bGravityOverrideInSimSpace=true
    bOverrideAngularDamping=true
    alphaBoolBlend {blendOption=Linear}
    boundBone {boneName=dyn_hat_l_04}
    componentLinearAccScale=<1.25 1.25 1.25>
    componentLinearVelScale=<1.25 1.25 1.25>
    physicsBodyDefinitions [{
      sphereCollisionRadius=0
      boundBone {boneName=dyn_hat_l_04}
      boxExtents=<5 15 15>
      constraintSetup {angularLimitsMax=<2 25 25>, angularLimitsMin=<-2 -25 -10>, angularTarget=<0 0 0>}
      localJointOffset=<1.5 0 0>
     }]
    relativeSpaceBone {boneName=root}
   }
   n28 = Anim Dynamics - Bone: dyn_hat_r_01(ComponentPose=@n27, Alpha=1, ExternalForce="0,0,0") {
    angularDampingOverride=0.98
    angularSpringConstant=300
    bAngularSpring=true
    bGravityOverrideInSimSpace=true
    bOverrideAngularDamping=true
    simulationSpace=RootRelative
    alphaBoolBlend {blendOption=Linear}
    boundBone {boneName=dyn_hat_r_01}
    componentLinearAccScale=<1.25 1.25 1.25>
    componentLinearVelScale=<1.25 1.25 1.25>
    physicsBodyDefinitions [{
      sphereCollisionRadius=0
      boundBone {boneName=dyn_hat_r_01}
      boxExtents=<5 20 20>
      constraintSetup {angularLimitsMax=<2 1 5>, angularLimitsMin=<-2 -25 -5>, angularTarget=<0 0 0>}
      localJointOffset=<1.5 0 0>
     }]
    relativeSpaceBone {boneName=root}
   }
   n29 = Anim Dynamics - Bone: dyn_hat_r_02(ComponentPose=@n28, Alpha=1, ExternalForce="0,0,0") {
    angularDampingOverride=0.98
    angularSpringConstant=300
    bAngularSpring=true
    bGravityOverrideInSimSpace=true
    bOverrideAngularDamping=true
    simulationSpace=RootRelative
    alphaBoolBlend {blendOption=Linear}
    boundBone {boneName=dyn_hat_r_02}
    componentLinearAccScale=<1.25 1.25 1.25>
    componentLinearVelScale=<1.25 1.25 1.25>
    physicsBodyDefinitions [{
      sphereCollisionRadius=0
      boundBone {boneName=dyn_hat_r_02}
      boxExtents=<5 15 15>
      constraintSetup {angularLimitsMax=<2 1 10>, angularLimitsMin=<-2 -25 -10>, angularTarget=<0 0 0>}
      localJointOffset=<1.5 0 0>
     }]
    relativeSpaceBone {boneName=root}
   }
   n30 = Anim Dynamics - Bone: dyn_hat_r_03(ComponentPose=@n29, Alpha=1, ExternalForce="0,0,0") {
    angularDampingOverride=0.98
    angularSpringConstant=300
    bAngularSpring=true
    bGravityOverrideInSimSpace=true
    bOverrideAngularDamping=true
    simulationSpace=RootRelative
    alphaBoolBlend {blendOption=Linear}
    boundBone {boneName=dyn_hat_r_03}
    componentLinearAccScale=<1.25 1.25 1.25>
    componentLinearVelScale=<1.25 1.25 1.25>
    physicsBodyDefinitions [{
      sphereCollisionRadius=0
      boundBone {boneName=dyn_hat_r_03}
      boxExtents=<5 15 15>
      constraintSetup {angularLimitsMax=<2 25 25>, angularLimitsMin=<-2 -25 -10>, angularTarget=<0 0 0>}
      localJointOffset=<1.5 0 0>
     }]
    relativeSpaceBone {boneName=root}
   }
   n31 = Anim Dynamics - Bone: dyn_hat_r_04(ComponentPose=@n30, Alpha=1, ExternalForce="0,0,0") {
    angularDampingOverride=0.98
    angularSpringConstant=300
    bAngularSpring=true
    bGravityOverrideInSimSpace=true
    bOverrideAngularDamping=true
    alphaBoolBlend {blendOption=Linear}
    boundBone {boneName=dyn_hat_r_04}
    componentLinearAccScale=<1.25 1.25 1.25>
    componentLinearVelScale=<1.25 1.25 1.25>
    physicsBodyDefinitions [{
      sphereCollisionRadius=0
      boundBone {boneName=dyn_hat_r_04}
      boxExtents=<5 15 15>
      constraintSetup {angularLimitsMax=<2 25 25>, angularLimitsMin=<-2 -25 -10>, angularTarget=<0 0 0>}
      localJointOffset=<1.5 0 0>
     }]
    relativeSpaceBone {boneName=root}
   }
   n32 = Anim Dynamics - Bone: dyn_l_grenade_01(ComponentPose=@n31, Alpha=1, ExternalForce="0,0,0") {
    angularDampingOverride=0.998
    angularSpringConstant=75
    bAngularSpring=true
    bGravityOverrideInSimSpace=true
    bOverrideAngularDamping=true
    bUseGravityOverride=true
    simulationSpace=RootRelative
    alphaBoolBlend {blendOption=Linear}
    boundBone {boneName=dyn_l_grenade_01}
    componentAppliedLinearAccClamp=<3000 3000 3000>
    componentLinearAccScale=<0.4 0.5 0.4>
    componentLinearVelScale=<0.4 0.5 0.4>
    gravityOverride {z=-720}
    physicsBodyDefinitions [{
      sphereCollisionRadius=0
      boundBone {boneName=dyn_l_grenade_01}
      boxExtents=<12 10 10>
      constraintSetup {angularLimitsMax=<1 -1 15>, angularLimitsMin=<-45 -60 -15>, angularTarget=<0 0 0>}
      localJointOffset=<3 0 0>
     }]
    relativeSpaceBone {boneName=dyn_spring_04}
   }
   n33 = Anim Dynamics - Bone: dyn_r_grenade_01(ComponentPose=@n32, Alpha=1, ExternalForce="0,0,0") {
    angularDampingOverride=0.998
    angularSpringConstant=100
    bAngularSpring=true
    bGravityOverrideInSimSpace=true
    bOverrideAngularDamping=true
    bUseGravityOverride=true
    simulationSpace=RootRelative
    alphaBoolBlend {blendOption=Linear}
    boundBone {boneName=dyn_r_grenade_01}
    componentAppliedLinearAccClamp=<3000 3000 3000>
    componentLinearAccScale=<0.5 0.6 0.5>
    componentLinearVelScale=<0.4 0.6 0.4>
    gravityOverride {z=-720}
    physicsBodyDefinitions [{
      sphereCollisionRadius=0
      boundBone {boneName=dyn_r_grenade_01}
      boxExtents=<15 10 10>
      constraintSetup {angularLimitsMax=<1 -1 15>, angularLimitsMin=<-45 -60 -15>, angularTarget=<0 0 0>}
      localJointOffset=<4 0 0>
     }]
    relativeSpaceBone {boneName=dyn_spring_04}
   }
   n34 = Anim Dynamics - Bone: dyn_r_grenade_pin(ComponentPose=@n33, Alpha=1, ExternalForce="0,0,0") {
    angularDampingOverride=0.999
    bGravityOverrideInSimSpace=true
    bOverrideAngularDamping=true
    bUseGravityOverride=true
    alphaBoolBlend {blendOption=Linear}
    boundBone {boneName=dyn_r_grenade_pin}
    componentAppliedLinearAccClamp=<3000 3000 3000>
    componentLinearAccScale=<0.3 0.3 0.3>
    componentLinearVelScale=<0.3 0.3 0.3>
    gravityOverride {z=-980}
    physicsBodyDefinitions [{
      sphereCollisionRadius=0
      boundBone {boneName=dyn_r_grenade_pin}
      boxExtents=<5 10 10>
      constraintSetup {angularLimitsMax=<20 5 20>, angularLimitsMin=<-20 -45 -20>, angularTarget=<0 0 0>}
      localJointOffset=<3 0 0>
     }]
   }
   n35 = Anim Dynamics - Bone: dyn_l_grenade_pin(ComponentPose=@n34, Alpha=1, ExternalForce="0,0,0") {
    angularDampingOverride=0.999
    bGravityOverrideInSimSpace=true
    bOverrideAngularDamping=true
    bUseGravityOverride=true
    alphaBoolBlend {blendOption=Linear}
    boundBone {boneName=dyn_l_grenade_pin}
    componentAppliedLinearAccClamp=<3000 3000 3000>
    componentLinearAccScale=<0.3 0.3 0.3>
    componentLinearVelScale=<0.3 0.3 0.3>
    gravityOverride {z=-980}
    physicsBodyDefinitions [{
      sphereCollisionRadius=0
      boundBone {boneName=dyn_l_grenade_pin}
      boxExtents=<5 10 10>
      constraintSetup {angularLimitsMax=<20 5 20>, angularLimitsMin=<-20 -45 -20>, angularTarget=<0 0 0>}
      localJointOffset=<3 0 0>
     }]
   }
   n36 = RigidBody!(ComponentPose=Local To Component(LocalPose=Local Space Ref Pose()), Alpha:float, ExternalForce) -> (Pose) {
    bClampLinearTranslationLimitToRefPose=true
    bOverrideWorldGravity=true
    overridePhysicsAsset=&F_MED_Cotton_Candy_Pack_Physics_Box
    simulationSpace=BaseBoneSpace
    alphaBoolBlend {blendOption=Linear}
    baseBoneRef {boneName=pack}
    componentLinearAccScale=<0.3 0.3 0.5>
    componentLinearVelScale=<0.2 0.2 0.3>
    overrideWorldGravity {z=-980}
   }
   n37 = "Transform (Modify) Bone - Bone: dyn_spring_MSTR"(
    ComponentPose=@n36
    Alpha=1
    Translation="0,0,0"
    Rotation=Get Spring_angle() -> (Spring_angle:Rotator)
    Scale=Get Surprise_Scale() -> (Surprise_Scale:Vector)
   ) {
    rotationMode=BMM_Additive
    scaleMode=BMM_Replace
    scaleSpace=BCS_BoneSpace
    alphaBoolBlend {blendOption=Linear}
    boneToModify {boneName=dyn_spring_MSTR}
   }
   n38 = Control Rig!(Source=Component To Local(ComponentPose=@n37), bAlphaBoolEnabled:bool=True) -> (Pose) {
    alphaInputType=Bool
    alphaBoolBlend {blendInTime=0.4, blendOption=Linear, blendOutTime=0.4}
    controlRigAssetReference {blueprintRigClass=/BRCosmetics/Accessories/FORT_Backpacks/Backpack_F_MED_Cotton_Candy/Meshes/F_MED_Cotton_Candy_Pack_CtrlRig/F_MED_Cotton_Candy_Pack_CtrlRig.F_MED_Cotton_Candy_Pack_CtrlRig_C}
   }
   Output Pose(Result=Blend Poses by bool(BlendPose_0=Component To Local(ComponentPose=@n35), BlendPose_1=@n38, BlendTime_1=0.15, bActiveValue=@n12) {blendType=Linear}) {layerGroup=None}
   Get Active_Kill() -> (Active_Kill:bool)
   Get bIsCrouching()
  ]}
 Close FX {nodes [Close FX() {is=title K2Node_FunctionEntry}]}
 EventGraph {nodes [
   Event BlueprintInitializeAnimation()
   > Set Skel_Mesh(Skel_Mesh=GetOwnerSkeletalMeshComponent())
   Event OnCharacterCustomizationLoaded()
   > n1 = Cast To FortPlayerPawn(Object=TryGetFortPawn())
   > n2 = WatchCosmeticStat!(
    execute:Exec
    Pawn:Fort Player Pawn Object Reference=@n1
    StatTag:Fort Stat Manager Tag Structure=(TagName="StatManager.AthenaKills")
    DefaultValue:int=0
    bBroadcastInitialValue:bool=false
    InitialValueBroadcastDelay:float=-1
    ChangeValueBroadcastDelay:float=-1
    bConsumeWaitingNotifications:bool=true
   ) -> (then:Exec, OnStatChanged:Exec, StatName:Name, StatValue:int) {is=title K2Node_AsyncAction}
   Set Did_Kill(execute=@n2.OnStatChanged, Did_Kill:bool=false) -> (Output_Get:bool)
   > n3 = Set KillCount(KillCount:double="To Float (Integer)"(InInt=@n2.StatValue)) -> (Output_Get:double)
   > Branch(Condition="float > float"(A=@n3, B=Get Last_KillCount() -> (Last_KillCount:double)) {is=title K2Node_CallFunction})
   > Set Active_Kill(Active_Kill:bool=true) -> (Output_Get:bool)
   > Set Last_KillCount(Last_KillCount:double=@n3) -> (Output_Get:double)
   n4 = Event BlueprintUpdateAnimation()
   > n5 = Sequence!(execute) -> (then_0, then_1, then_2:Exec, then_3:Exec, then_4:Exec, then_5:Exec, then_6:Exec)
   > Branch(Condition=Get Active_Kill())
   > Delay(Duration=RandomFloatInRange(Min=0.5, Max=3))
   > Set Surprise_Out(Surprise_Out:bool=true) -> (Output_Get:bool)
   > Set Active_Kill(Active_Kill=false)
   > n6 = Branch(Condition=Get Surprise_Out())
   > Delay(Duration=0.2)
   > Set physics_out(physics_out:bool=true) -> (Output_Get:bool)
   Set physics_out(execute=@n6.else, physics_out=false)
   n7 = Branch(execute=@n5.then_1, Condition=Get Surprise_Out())
   > Set Surprise_Alpha(Surprise_Alpha:double=1) -> (Output_Get:double)
   > n8 = Set Surprise_Timer(Surprise_Timer:double=float + float(A=Get Surprise_Timer() -> (Surprise_Timer:double), B=1) {is=title K2Node_CommutativeAssociativeBinaryOperator}) -> (Output_Get:double)
   > Set Surprise_Scale(Surprise_Scale:Vector=VInterpTo_Constant!(Current:Vector=Get Surprise_Scale(), Target:Vector="1.1,1.1,1.1", DeltaTime:float=@n4.DeltaTimeX, InterpSpeed:float=5) -> (ReturnValue:Vector)) -> (Output_Get:Vector)
   > n9 = Branch(Condition="Equal (Float)"!(A:double=@n8, B:double=Get Surprise_Length_Out() -> (Surprise_Length_Out:double)) -> (ReturnValue:bool) {is=title K2Node_CallFunction})
   > Set Surprise_Out(Surprise_Out=false)
   > Set Surprise_Timer(Surprise_Timer=0)
   > Set Surprise_Alpha(Surprise_Alpha=0)
   > n10 = Call Close FX!(execute:Exec, self:F MED Cotton Candy Pack Anim BP Object Reference) -> (then:Exec) {is=title K2Node_CallDelegate}
   Do Once(execute=@n9.else, Reset=@n10.then)
   > Call Surprise FX!(execute:Exec, self:F MED Cotton Candy Pack Anim BP Object Reference) -> (then:Exec) {is=title K2Node_CallDelegate}
   Set Surprise_Scale(
    execute=@n7.else
    Surprise_Scale=VInterpTo_Constant(Current=Get Surprise_Scale(), Target="0.2,0.6,0.6", DeltaTime=@n4.DeltaTimeX, InterpSpeed=8)
   )
   > Set Surprise_Length_Out(Surprise_Length_Out:double=600) -> (Output_Get:double)
   > Set physics_out(physics_out=false)
   n11 = Branch(execute=@n5.then_4, Condition=Get physics_out())
   > Do Once(Reset=@n13.then)
   > Set Force(Force:bool=true) -> (Output_Get:bool)
   > Delay(Duration=0.2)
   > Set Force(Force=false)
   Set Force(execute=@n11.else, Force=false)
   n12 = Branch(execute=@n5.then_5, Condition=Get Force() -> (Force:bool))
   > Set Surprise Force(Surprise Force:Vector="0, 0,100") -> (Output_Get:Vector)
   Set Surprise Force(execute=@n12.else, Surprise Force="0,0,-50")
   n13 = Branch(execute=@n5.then_6, Condition=NOT Boolean(A=Get Surprise_Out()))
   n14 = RandomRotatorFunction!(
    execute=@n5.then_2
    GameTimeInSeconds:double=GetGameTimeInSeconds()
    PerlinScale:double=1
    Frequency:Rotator="2.2, 0,1.3"
    Amplitude:Rotator="7,0,7"
    Offset:Rotator="-3,0,-5"
    TimeOffset:double=0
   ) -> (then, RandomRotator:Rotator)
   > Set Spring_angle(Spring_angle:Rotator=@n14) -> (Output_Get:Rotator)
   n15 = Gravity Override State Function DEPRECATED(
    execute=@n5.then_3
    CharacterPartAnimInstance=Self-Reference()
    GameTimeInSeconds=GetGameTimeInSeconds()
    GravityOverrideParams=Get Gravity Override Params()
   )
   > Set GravityOverride(GravityOverride=@n15)
   n16 = Branch()
   > Sequence!(execute) -> (then_0, then_1)
   > Set initial transfrom(initial transfrom:double=1) -> (Output_Get:double)
   > Delay(Duration=0.3)
   > Set initial transfrom(initial transfrom=0)
   Set initial transfrom(execute=@n16.else, initial transfrom=0)
   TryGetPawnOwner()
   Delay(Duration=0.3)
   n17 = GetGameTimeInSeconds()
   n18 = MapRangeClamped(Value="Cos (Degrees)"!(A:double=float * float(A=@n17, B=35)) -> (ReturnValue:double), InRangeA=-3, InRangeB=3, OutRangeA=-20, OutRangeB=20)
   MakeRotator(Pitch=@n18, Yaw=@n18)
   n19 = Get random_stream() -> (random_stream:Random Stream Structure)
   vector * vector(
    A:Vector=MakeVector(
     Y=0
     Z=RandomFloatInRangeFromStream!(Stream:"Random Stream Structure (by ref)"=@n19, Min:float=-2000, Max:float=-5000) -> (ReturnValue:float)
    )
    B:Vector="1,1,1"
   ) -> (ReturnValue:Vector) {is=title K2Node_CommutativeAssociativeBinaryOperator}
   Do Once Then Do!(execute:Exec, Reset:Exec, StartClosed:bool) -> (DoOnce:Exec, ThenDo:Exec) {is=title K2Node_MacroInstance}
   Get bIsTargeting()
   Get bIsTargeting()
   Get Surprise_Out()
   Set arm_scale(arm_scale:Vector=VInterpTo_Constant(Current=Get arm_scale() -> (arm_scale:Vector), Target="1.1,1.1,1.1", DeltaTime=@n4.DeltaTimeX, InterpSpeed=15)) -> (Output_Get:Vector)
   Set random_stream(random_stream:Random Stream Structure=MakeRandomStream!(InitialSeed:int=0) -> (ReturnValue:Random Stream Structure)) -> (Output_Get:Random Stream Structure)
  ]}
 Surprise FX {nodes [Surprise FX() {is=title K2Node_FunctionEntry}]}
]
variables [
 Active_Kill {type=boolean}
 arm_scale {type=Vector}
 Close FX {default=()}
 CountdownActive {type=boolean}
 Did_Kill {type=boolean}
 Force {type=boolean}
 Gravity Override Params {
  type=GravityOverrideParamsStruct
  default {
   emoteOrMelee {gravityOverride=<-25 0 -100>, windAmplitude=<25 0 100>, windFrequency=<5 0 0>, windOffset=<-25 0 -100>}
   front_End {gravityOverride=<-25 0 -100>, windAmplitude=<25 0 100>, windFrequency=<5 0 0>, windOffset=<-25 0 -100>}
   onGround_Moving {gravityOverride=<-25 0 -100>, windAmplitude=<25 0 100>, windFrequency=<5 0 0>, windOffset=<-25 0 -100>}
   onGround_Standing {gravityOverride=<-25 0 -100>, windAmplitude=<25 0 100>, windFrequency=<5 0 0>, windOffset=<-25 0 -100>}
  }
 }
 GravityOverride {type=Vector}
 initial transfrom {type=number}
 KillCount {type=number}
 Last_KillCount {type=number}
 physics_out {type=boolean}
 random_stream {type=RandomStream, default {initialSeed=0, seed=0}}
 Skel_Mesh {type=SkeletalMeshComponent}
 Spring_angle {type=Rotator}
 Surprise Force {type=Vector}
 Surprise FX {default=()}
 Surprise_Alpha {type=number}
 Surprise_Length_Out {type=number}
 Surprise_Out {type=boolean}
 Surprise_Scale {type=Vector}
 Surprise_Timer {type=number}
]
